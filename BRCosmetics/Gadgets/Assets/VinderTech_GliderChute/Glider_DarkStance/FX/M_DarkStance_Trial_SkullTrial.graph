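# Translucent unlit Niagara ribbon material for the skull trail effect of the DarkStance glider.
# Samples the skull texture three times at slightly shifted, smoke-distorted UVs for an RGB-split look, with dynamic parameters driving skull and noise scroll.
# Adds tinted rainbow color, multiplies by blurred noise, then applies desaturation and particle color for emissive.
# Opacity combines color brightness, sine edge masks, a lead mask, depth fade on non-mobile deferred paths and optional near-camera fade.
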Material /BRCosmetics/Gadgets/Assets/VinderTech_GliderChute/Glider_DarkStance/FX/M_DarkStance_Trial_SkullTrial
graph {
 properties {bUsedWithNiagaraRibbons=true, blendMode=BLEND_Translucent, pixelDepthOffsetMode=PDOM_Legacy, shadingModel=MSM_Unlit, usedShadingModels=Unlit}
 nodes [
  n1 = ParticleColor()
  n2 = Constant2Vector() {g=-0.05}
  n3 = Constant() {r=0.02}
  n4 = Multiply(
   A=TextureCoordinate()
   B=ComponentMask(None=VectorParameter() {parameterName=TexAUVScale, defaultValue=<rgba 2.5 1.2 0 0>}.RGB) {g=true, r=true}
  )
  n5 = TextureSample(UVs=Panner(Coordinate=TextureCoordinate() {uTiling=2}) {speedY=0.5}) {texture=&T_Inky_Smoke_Tile}
  n6 = ScalarParameter() {defaultValue=0.1, parameterName=DisturbIntensity}
  n7 = Add(A=Add(A=@n4, B=Add(A=@n2, B=@n3)), B=Multiply(A=@n5.R, B=@n6) {constB=0.13})
  n8 = DynamicParameter() {defaultValue=<rgba 0 0 0.2 1>, paramNames [Skull_offset, Noise offset, ColorTex_Offset, Param4]}
  n9 = Add(A=Add(A=@n4, B=@n2), B=Multiply(A=@n5.R, B=@n6) {constB=0.13})
  n10 = Add(A=Add(A=@n4, B=Subtract(A=@n2, B=@n3)), B=Multiply(A=@n5.R, B=@n6) {constB=0.13})
  n11 = MaterialFunctionCall(
   R=TextureSampleParameter2D(UVs=Panner(Coordinate=@n7, Time=@n8.Skull_offset) {speedX=1}) {parameterName=TexA, texture=/BRCosmetics/Weapons/FORT_Melee/Pickaxe_DarkStance/FX/T_PickaxeDarkStance_SkullTrial}.R
   G=TextureSampleParameter2D(UVs=Panner(Coordinate=@n9, Time=@n8.Skull_offset) {speedX=1}) {parameterName=TexA, texture=&T_PickaxeDarkStance_SkullTrial}.R
   B=TextureSampleParameter2D(UVs=Panner(Coordinate=@n10, Time=@n8.Skull_offset) {speedX=1}) {parameterName=TexA, texture=&T_PickaxeDarkStance_SkullTrial}.R
  ) {materialFunction=&AppendMany}
  n12 = ScalarParameter() {defaultValue=1, parameterName=AddAditionalNoise}
  n13 = TextureCoordinate()
  n14 = ComponentMask(None=VectorParameter() {parameterName=ColorTexUVScale, defaultValue=<rgba 1 1 0 0>}.RGB) {g=true, r=true}
  n15 = Panner(Coordinate=Add(A=Multiply(A=@n13, B=@n14)) {constB=0}, Time=ScalarParameter() {defaultValue=0.2, parameterName=ColorTexOffset}) {speedX=0.5}
  n16 = Power(Base=Multiply(
    A=TextureSample(UVs=@n15) {texture=&T_Tile_RainbowColor}.RGB
    B=Constant3Vector() {constant=<rgba 0.084 0.4948 0.3327 1>}
   ) {constB=0.5}) {constExponent=0.7}
  n17 = Panner(
   Coordinate=TextureCoordinate()
   Time=DynamicParameter() {defaultValue=<rgba 0 0 0.2 1>, paramNames [Skull_offset, Noise offset, ColorTex_Offset, Param4]}.Noise offset
  ) {speedX=-0.5}
  n18 = Multiply(
   A=Add(A=Multiply(A=@n11.RGB) {constB=2}, B=LinearInterpolate(B=@n16, Alpha=@n12)) {constB=0}
   B=Power(
    Base=TextureSample(UVs=@n17) {samplerType=SAMPLERTYPE_Grayscale, texture=&LowResBlurredNoise}.R
    Exp=ScalarParameter() {defaultValue=2, parameterName=NoisePower}
   )
  )
  n19 = Desaturation(None=@n18, Fraction=ScalarParameter() {parameterName=DesColor})
  Reroute(None=MaterialFunctionCall(Emissive=Multiply(A=@n1.RGB, B=@n19)) {materialFunction=&MF_FX_EmissiveAdjustments}.Result)
  n20 = ComponentMask(None=TextureCoordinate()) {g=true}
  n21 = Multiply(
   A=Power(Base=MaterialFunctionCall(Vector 3=@n18) {materialFunction=/Engine/Functions/Engine_MaterialFunctions02/Utility/VectorLength}.V3 Length) {constExponent=0.5}
   B=Power(Base=Sine(None=@n20) {period=2}, Exp=ScalarParameter() {defaultValue=6, parameterName=TopBottomMaskPower}) {constExponent=6}
  )
  n22 = ComponentMask(None=TextureCoordinate()) {r=true}
  n23 = ComponentMask(None=TextureCoordinate()) {r=true}
  n24 = Multiply(A=Subtract(A=OneMinus(None=@n23)) {constB=0.4}) {constB=5}
  n25 = Multiply(
   A=Multiply(A=Multiply(A=@n1.A, B=@n21), B=Sine(None=@n22) {period=2})
   B=LinearInterpolate(B=Saturate(None=@n24), Alpha=ScalarParameter() {defaultValue=1, parameterName=UseLeadMask}) {constA=1}
  )
  n26 = DepthFade(Opacity=@n25) {fadeDistanceDefault=10}
  n27 = MaterialFunctionCall(
   In=MaterialFunctionCall(Distance=ScalarParameter() {defaultValue=300, parameterName=NearCameraFade}) {materialFunction=&Near_Camera_Fade}.Result
   Contrast=ScalarParameter() {defaultValue=0.35, parameterName=Contrast}
  ) {materialFunction=&CheapContrast}
  n28 = ScalarParameter() {parameterName=UseNearCameraFade}
  Reroute(None=Multiply(A=ShadingPathSwitch(Default=@n26, Deferred=@n26, Forward=@n25, Mobile=@n25), B=LinearInterpolate(B=@n27.Result, Alpha=@n28) {constA=1}))
 ]
}
