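# Niagara effect for the Dark Stance glider: sparks, smoke and flash bursts, lingering mesh smoke, drifting skulls and a light.
# Emitters follow the skeletal mesh, using partitioned offsets to mirror effects on both sides.
# System-level interpolated values drive smoke scale, tilt and per-side fading, plus skull spawning and light brightness.
# User.VariantSwitch blends green to purple; front-end preview suppresses several bursts, while duplicate burst emitters fire from a DoOnce trigger.
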NiagaraSystem /BRCosmetics/Gadgets/Assets/VinderTech_GliderChute/Glider_DarkStance/FX/NS_Glider_DarkStance
dependencies {
 usedDataInterfaces [
  &NiagaraDataInterfaceArrayFloat3
  &NiagaraDataInterfaceCurve
  &NiagaraDataInterfaceColorCurve
  &NiagaraDataInterfaceSpriteRendererInfo
  &NiagaraDataInterfaceArrayFloat
  &NiagaraDataInterfacePlatformSet
  &NiagaraDataInterfaceSkeletalMesh
 ]
 usedDynamicInputs [
  &Subtract_Float
  &Abs_Float
  &DotProduct
  /Niagara/DynamicInputs/Multiply/ScaleAndBiasFloat
  &Ceil_Float
  &SetBoolByFloatComparison
  &ClampFloat
  &Boolean_AndOperation
 ]
 usedModules [
  &self:SystemSpawnScript.SystemScriptSource.SystemScriptGraph.NiagaraNodeAssignment_0.SetVariables_#1
  &self:SystemSpawnScript.SystemScriptSource.SystemScriptGraph.NiagaraNodeAssignment_1.SetVariables_#2
  &self:SystemSpawnScript.SystemScriptSource.SystemScriptGraph.NiagaraNodeAssignment_2.SetVariables_#3
  &self:SystemSpawnScript.SystemScriptSource.SystemScriptGraph.NiagaraNodeAssignment_3.SetVariables_#4
  &SystemState
  &self:SystemSpawnScript.SystemScriptSource.SystemScriptGraph.NiagaraNodeAssignment_4.SetVariables_#5
  &DoOnce
 ]
}
emitters {
 Burst_Sparks {
  simTarget=GPUComputeSim
  emitterUpdateScript [
   &EmitterState(Life Cycle Mode=::NewEnumerator1, Loop Behavior=::NewEnumerator1, Loop Delay=$System.Delay, Loop Duration=0.3, UseLoopDelay=true)
   &SpawnBurst_Instantaneous(
    Spawn Count=80
    Spawn Probability=&MakeFloatFromInt(Int=&MakeCustomIntFromBool(Bool=$User.bIsFrontEndPreview, True INT=0, False INT=1))
    Use Spawn Probability=true
    UseLoopCountLimit=false
   )
  ]
  particleSpawnScript [
   &PartitionParticles()
   &InitializeParticle(
    Color=&LerpLinearColors(StartColor=<rgba 0.5647 5 1.5996 1>, EndColor=<rgba 5.5647 1.1293 10 1>, LerpFactor=$User.VariantSwitch)
    Lifetime Max=1.5
    Lifetime Mode=::NewEnumerator1
    Sprite Size Max=<1 10>
    Sprite Size Min=<1 5>
    Sprite Size Mode=::NewEnumerator2
    UsePositionOffset=false
    Write Material Random=true
   )
   &SkeletalMeshLocation(
    Bone Sampling Mode=::NewEnumerator4
    MaskSpawnGroup=false
    Orientation Sampling=::NewEnumerator1
    Sampled Position Offset=&AddVectorToPosition(
     Position=&SelectVectorFromArray(
      Direct Array Index=$Particles.PartitionParticles.Partition
      Vector Selection Array=DataInterface() {propertyValues {FloatData [<-75 0 0>, <75 0 0>]}}
     )
     Vector=<0 0 0>
    )
    Skeletal Mesh=$System.Skeletal Mesh
   )
   &ShapeLocation(MaskSpawnGroup=false, Override Seed=false, Sphere Radius=50, UseOwnerScale=true)
   &AddVelocity(
    Cone Angle=50
    Cone Axis=&SelectVectorFromArray(
     Direct Array Index=$Particles.PartitionParticles.Partition
     Vector Selection Array=DataInterface() {propertyValues {FloatData [<1 -0.8 0.5>, <-1 -0.8 0.5>]}}
    )
    Module.UseConeFalloff=false
    Module.UseVelDistribution=false
    Override Seed=false
    Velocity Mode=::NewEnumerator2
    Velocity Speed=&RandomRangeFloat(Minimum=150, Maximum=600)
   )
  ]
  particleUpdateScript [
   &ParticleState()
   /SetVariables(Particles.Glow03_Color=&MakeLinearColorFromVectorAndFloat(
     "Vector (RGB)"=&MakeVectorFromLinearColorRGB(Color=&MultiplyLinearColor(
       Linear Color=&LerpLinearColors(StartColor=<rgba 0.0335 1 0.2591 1>, EndColor=<rgba 1.0335 0.0671 2 1>, LerpFactor=$User.VariantSwitch)
       Scale Factor=&MakeVector4_FromFloat(Float=1)
      ))
     "Float (Alpha)"=&Multiply_Float(A=0.2, B=1)
    )) {isSetParameters=true, module=SetVariables_#6}
   /SetVariables(Particles.Glow03_Size=&Multiply_Vector2D_ByFloat(
     Float=&FloatFromCurve(
      FloatCurve=DataInterface() {propertyValues {Curve {keys [
          <at 0 0 RCIM_Cubic>
          <at 0.2 1.3 RCIM_Cubic>
          {arriveTangent=-0.1423, interpMode=RCIM_Cubic, leaveTangent=-0.1423, tangentMode=RCTM_User, time=1, value=0}
         ]}}}
      Scale Curve=10
     )
     Vector2D=<2 5>
    )) {isSetParameters=true, module=SetVariables_#7}
   &ScaleColor(
    Scale Alpha=&FloatFromCurve(FloatCurve=DataInterface() {propertyValues {Curve {keys [<at 0.3 1>,<at 0.4 1>,<at 1 1>]}}})
    ScaleA=true
    ScaleRGB=true
   )
   &ScaleSpriteSize(Scale Factor=&Vector2DFromFloat(Value=&FloatFromCurve(
      FloatCurve=DataInterface() {propertyValues {Curve {keys [
          <at 0 0 RCIM_Cubic>
          {interpMode=RCIM_Cubic, leaveTangent=-3.0296, tangentMode=RCTM_Break, time=0.2, value=1.2}
          {arriveTangent=-1.8107, interpMode=RCIM_Cubic, leaveTangent=-1.8107, tangentMode=RCTM_User, time=1, value=0}
         ]}}}
      Scale Curve=2
     ))) {module=ScaleSpriteSize001}
   &CurlNoiseForce(Noise Frequency=500, Noise Strength=1000, Pan Noise=false)
   &Drag(Use Linear Drag=true, Use Rotational Drag=true)
   &SolveForcesAndVelocity(Clamp Velocity=false, Limit Acceleration=false)
  ]
  rendererClasses [&NiagaraSpriteRendererProperties]
  renderers [
   {rendererClass=&NiagaraSpriteRendererProperties}
   {rendererClass=&NiagaraSpriteRendererProperties, rendererIndex=1}
  ]
 }
 Burst_Sparks001 {
  simTarget=GPUComputeSim
  emitterUpdateScript [
   &EmitterState(Delay First Loop Only=true, Life Cycle Mode=::NewEnumerator1, Loop Delay=$System.Delay, Loop Duration=0.3, UseLoopDelay=true)
   &SpawnPerFrame(Spawn:NiagaraBool=$System.DoOnce.Execute, Spawn Count=20, Use Spawn Probability=false)
  ]
  particleSpawnScript [
   &PartitionParticles()
   &InitializeParticle(
    Color=&LerpLinearColors(StartColor=<rgba 0.5647 5 1.5996 1>, EndColor=<rgba 5.5647 1.1293 10 1>, LerpFactor=$User.VariantSwitch)
    Lifetime Max=1
    Lifetime Min=0.8
    Lifetime Mode=::NewEnumerator1
    Sprite Size Max=<1 10>
    Sprite Size Min=<1 5>
    Sprite Size Mode=::NewEnumerator2
    UsePositionOffset=false
    Write Material Random=true
   )
   &SkeletalMeshLocation(
    Bone Sampling Mode=::NewEnumerator4
    MaskSpawnGroup=false
    Orientation Sampling=::NewEnumerator1
    Sampled Position Offset=&AddVectorToPosition(
     Position=&SelectVectorFromArray(
      Direct Array Index=$Particles.PartitionParticles.Partition
      Vector Selection Array=DataInterface() {propertyValues {FloatData [<-75 0 0>, <75 0 0>]}}
     )
     Vector=<0 0 0>
    )
    Skeletal Mesh=$System.Skeletal Mesh
   )
   &ShapeLocation(MaskSpawnGroup=false, Override Seed=false, Sphere Radius=50, UseOwnerScale=true)
   &AddVelocity(
    Cone Angle=70
    Cone Axis=&SelectVectorFromArray(
     Direct Array Index=$Particles.PartitionParticles.Partition
     Vector Selection Array=DataInterface() {propertyValues {FloatData [<1 -0.8 0.5>, <-1 -0.8 0.5>]}}
    )
    Module.UseConeFalloff=false
    Module.UseVelDistribution=false
    Override Seed=false
    Velocity Mode=::NewEnumerator2
    Velocity Speed=&RandomRangeFloat(Minimum=300, Maximum=600)
   )
  ]
  particleUpdateScript [
   &ParticleState()
   /SetVariables(Particles.Glow03_Color=&MakeLinearColorFromVectorAndFloat(
     "Vector (RGB)"=&MakeVectorFromLinearColorRGB(Color=&MultiplyLinearColor(
       Linear Color=&LerpLinearColors(StartColor=<rgba 0.0335 1 0.2591 1>, EndColor=<rgba 1.0335 0.0671 2 1>, LerpFactor=$User.VariantSwitch)
       Scale Factor=&MakeVector4_FromFloat(Float=1)
      ))
     "Float (Alpha)"=&Multiply_Float(A=0.2, B=1)
    )) {isSetParameters=true, module=SetVariables_#6}
   /SetVariables(Particles.Glow03_Size=&Multiply_Vector2D_ByFloat(
     Float=&FloatFromCurve(
      FloatCurve=DataInterface() {propertyValues {Curve {keys [
          <at 0 0 RCIM_Cubic>
          <at 0.2 1.3 RCIM_Cubic>
          {arriveTangent=-0.1423, interpMode=RCIM_Cubic, leaveTangent=-0.1423, tangentMode=RCTM_User, time=1, value=0}
         ]}}}
      Scale Curve=10
     )
     Vector2D=<2 5>
    )) {isSetParameters=true, module=SetVariables_#7}
   &ScaleColor(
    Scale Alpha=&FloatFromCurve(FloatCurve=DataInterface() {propertyValues {Curve {keys [<at 0.3 1>,<at 0.4 1>,<at 1 1>]}}})
    ScaleA=true
    ScaleRGB=true
   )
   &ScaleSpriteSize(Scale Factor=&Vector2DFromFloat(Value=&FloatFromCurve(
      FloatCurve=DataInterface() {propertyValues {Curve {keys [
          <at 0 0 RCIM_Cubic>
          {interpMode=RCIM_Cubic, leaveTangent=-3.0296, tangentMode=RCTM_Break, time=0.2, value=1.2}
          {arriveTangent=-1.8107, interpMode=RCIM_Cubic, leaveTangent=-1.8107, tangentMode=RCTM_User, time=1, value=0}
         ]}}}
      Scale Curve=2
     ))) {module=ScaleSpriteSize001}
   &CurlNoiseForce(Noise Frequency=500, Noise Strength=1000, Pan Noise=false)
   &Drag(Drag=0.5, Use Linear Drag=true, Use Rotational Drag=true)
   &SolveForcesAndVelocity(Clamp Velocity=false, Limit Acceleration=false)
  ]
  rendererClasses [&NiagaraSpriteRendererProperties]
  renderers [
   {rendererClass=&NiagaraSpriteRendererProperties}
   {rendererClass=&NiagaraSpriteRendererProperties, rendererIndex=1}
  ]
 }
 Burst_Sub_Smoke {
  emitterUpdateScript [
   &EmitterState(Life Cycle Mode=::NewEnumerator1, Loop Behavior=::NewEnumerator1, Loop Delay=$System.Delay, Loop Duration=0.5, UseLoopDelay=true)
   &SpawnBurst_Instantaneous(
    Spawn Count=6
    Spawn Probability=&MakeFloatFromInt(Int=&MakeCustomIntFromBool(Bool=$User.bIsFrontEndPreview, True INT=0, False INT=1))
    Use Spawn Probability=true
    UseLoopCountLimit=false
   )
  ]
  particleSpawnScript [
   &PartitionParticles()
   &InitializeParticle(
    Color=&LerpLinearColors(StartColor=<rgba 0.0053 0.04 0.0134 2>, EndColor=<rgba 0.046 0.012 0.08 2>, LerpFactor=$User.VariantSwitch)
    Lifetime Max=1.3
    Lifetime Min=0.7
    Lifetime Mode=::NewEnumerator1
    Sprite Rotation Mode=::NewEnumerator2
    Sprite Size Mode=::NewEnumerator1
    Sprite UV Mode=::NewEnumerator3
    Uniform Sprite Size Max=250
    Uniform Sprite Size Min=200
    UsePositionOffset=false
    Write Material Random=true
   )
   &SkeletalMeshLocation(
    Bone Sampling Mode=::NewEnumerator4
    MaskSpawnGroup=false
    Orientation Sampling=::NewEnumerator1
    Sampled Position Offset=&SelectVectorFromArray(
     Direct Array Index=$Particles.PartitionParticles.Partition
     Vector Selection Array=DataInterface() {propertyValues {FloatData [<-75 -20 0>, <75 -20 0>]}}
    )
    Sampled Position Offset Coordinate Space:ENiagaraSkelMeshTransforms=::NewEnumerator2
    Skeletal Mesh=$System.Skeletal Mesh
   )
   &AddVelocity(
    Cone Angle=30
    Cone Axis=&SelectVectorFromArray(
     Direct Array Index=$Particles.PartitionParticles.Partition
     Vector Selection Array=DataInterface() {propertyValues {FloatData [<-1 -0.8 0.5>, <1 -0.8 0.5>]}}
    )
    Module.UseConeFalloff=false
    Module.UseVelDistribution=false
    Override Seed=false
    Velocity Mode=::NewEnumerator2
    Velocity Speed=&RandomRangeFloat(Minimum=200, Maximum=600)
   ) {module=AddVelocity001}
   &CameraOffset(Camera Offset Amount=-50)
  ]
  particleUpdateScript [
   &ParticleState()
   &ScaleSpriteSize(Uniform Curve Sprite Scale=DataInterface() {propertyValues {Curve {keys [
        {arriveTangent=0.2792, interpMode=RCIM_Cubic, leaveTangent=0.2792, tangentMode=RCTM_User, time=0, value=1}
        {arriveTangent=0.637, tangentMode=RCTM_User, time=1, value=1.4}
       ]}}})
   &SpriteRotationRate(Rotation Rate=&RandomRangeFloat(Minimum=-150, Maximum=150))
   &Drag(Drag=2, Use Linear Drag=true, Use Rotational Drag=true)
   &SolveForcesAndVelocity(Clamp Velocity=false, Limit Acceleration=false)
   &ScaleColor(
    Scale Alpha=&FloatFromCurve(FloatCurve=DataInterface() {propertyValues {Curve {keys [{interpMode=RCIM_Cubic, tangentMode=RCTM_User, time=0, value=1}, {interpMode=RCIM_Cubic, tangentMode=RCTM_User, time=1, value=0}]}}})
    Scale RGB=&Multiply_VectorByFloat(
     Vector=<1 1 1>
     Float=&FloatFromCurve(FloatCurve=DataInterface() {propertyValues {Curve {keys [
          {arriveTangent=-12.6868, interpMode=RCIM_Cubic, leaveTangent=-12.6868, tangentMode=RCTM_User, time=0, value=20}
          {arriveTangent=-15, interpMode=RCIM_Cubic, leaveTangent=-15, time=0.5045, value=5.7672}
          {arriveTangent=-14.7826, interpMode=RCIM_Cubic, leaveTangent=-14.7826, tangentMode=RCTM_User, time=1, value=5}
         ]}}})
    )
    ScaleA=true
    ScaleRGB=true
   )
   &SubUVAnimation(
    End Frame Range Override=64
    "Sprite Renderer "=DataInterface() {propertyValues {SpriteRenderer=&self:Burst_Sub_Smoke_12.NiagaraSpriteRendererProperties_1}}
    SubUV Lookup Index Scale=0.5
    UseEndFrame=true
    UseStartFrame=true
   )
  ]
  rendererClasses [&NiagaraSpriteRendererProperties]
  renderers [{rendererClass=&NiagaraSpriteRendererProperties}]
 }
 Burst_Sub_Smoke001 {
  emitterUpdateScript [
   &EmitterState(Delay First Loop Only=true, Life Cycle Mode=::NewEnumerator1, Loop Delay=$System.Delay, Loop Duration=0.5, UseLoopDelay=true)
   &SpawnPerFrame(Spawn:NiagaraBool=$System.DoOnce.Execute, Spawn Count=4, Use Spawn Probability=false)
  ]
  particleSpawnScript [
   &PartitionParticles()
   &InitializeParticle(
    Color=&LerpLinearColors(StartColor=<rgba 0.0053 0.04 0.0134 0.5>, EndColor=<rgba 0.0402 0.0105 0.07 0.5>, LerpFactor=$User.VariantSwitch)
    Lifetime Max=1
    Lifetime Mode=::NewEnumerator1
    Sprite Rotation Mode=::NewEnumerator2
    Sprite Size Mode=::NewEnumerator1
    Sprite UV Mode=::NewEnumerator3
    Uniform Sprite Size Max=250
    Uniform Sprite Size Min=200
    UsePositionOffset=false
    Write Material Random=true
   )
   &SkeletalMeshLocation(
    Bone Sampling Mode=::NewEnumerator4
    MaskSpawnGroup=false
    Orientation Sampling=::NewEnumerator1
    Sampled Position Offset=&SelectVectorFromArray(
     Direct Array Index=$Particles.PartitionParticles.Partition
     Vector Selection Array=DataInterface() {propertyValues {FloatData [<-75 -20 0>, <75 -20 0>]}}
    )
    Sampled Position Offset Coordinate Space:ENiagaraSkelMeshTransforms=::NewEnumerator2
    Skeletal Mesh=$System.Skeletal Mesh
   )
   &AddVelocity(
    Cone Angle=30
    Cone Axis=&SelectVectorFromArray(
     Direct Array Index=$Particles.PartitionParticles.Partition
     Vector Selection Array=DataInterface() {propertyValues {FloatData [<-1 -0.8 0.5>, <1 -0.8 0.5>]}}
    )
    Module.UseConeFalloff=false
    Module.UseVelDistribution=false
    Override Seed=false
    Velocity Mode=::NewEnumerator2
    Velocity Speed=&RandomRangeFloat(Minimum=200, Maximum=600)
   ) {module=AddVelocity001}
   &CameraOffset(Camera Offset Amount=-50)
  ]
  particleUpdateScript [
   &ParticleState()
   &ScaleSpriteSize(Uniform Curve Sprite Scale=DataInterface() {propertyValues {Curve {keys [
        {arriveTangent=0.2792, interpMode=RCIM_Cubic, leaveTangent=0.2792, tangentMode=RCTM_User, time=0, value=1}
        {arriveTangent=0.637, tangentMode=RCTM_User, time=1, value=1.4}
       ]}}})
   &SpriteRotationRate(Rotation Rate=&RandomRangeFloat(Minimum=-150, Maximum=150))
   &Drag(Drag=2, Use Linear Drag=true, Use Rotational Drag=true)
   &SolveForcesAndVelocity(Clamp Velocity=false, Limit Acceleration=false)
   &ScaleColor(
    Scale Alpha=&FloatFromCurve(FloatCurve=DataInterface() {propertyValues {Curve {keys [{interpMode=RCIM_Cubic, tangentMode=RCTM_User, time=0, value=1}, {interpMode=RCIM_Cubic, tangentMode=RCTM_User, time=1, value=0}]}}})
    Scale RGB=&Multiply_VectorByFloat(
     Vector=<1 1 1>
     Float=&FloatFromCurve(FloatCurve=DataInterface() {propertyValues {Curve {keys [
          {arriveTangent=-12.6868, interpMode=RCIM_Cubic, leaveTangent=-12.6868, tangentMode=RCTM_User, time=0, value=20}
          {arriveTangent=-15, interpMode=RCIM_Cubic, leaveTangent=-15, time=0.5045, value=5.7672}
          {arriveTangent=-14.7826, interpMode=RCIM_Cubic, leaveTangent=-14.7826, tangentMode=RCTM_User, time=1, value=5}
         ]}}})
    )
    ScaleA=true
    ScaleRGB=true
   )
   &SubUVAnimation(
    End Frame Range Override=64
    "Sprite Renderer "=DataInterface() {propertyValues {SpriteRenderer=&self:Burst_Sub_Smoke001_9.NiagaraSpriteRendererProperties_1}}
    SubUV Lookup Index Scale=0.5
    UseEndFrame=true
    UseStartFrame=true
   )
  ]
  rendererClasses [&NiagaraSpriteRendererProperties]
  renderers [{rendererClass=&NiagaraSpriteRendererProperties}]
 }
 Flash {
  emitterUpdateScript [
   &EmitterState(Life Cycle Mode=::NewEnumerator1, Loop Behavior=::NewEnumerator1, Loop Delay=$System.Delay, Loop Duration=1, UseLoopDelay=true)
   &SpawnBurst_Instantaneous(
    Spawn Count=2
    Spawn Probability=&MakeFloatFromInt(Int=&MakeCustomIntFromBool(Bool=$User.bIsFrontEndPreview, True INT=0, False INT=1))
    Use Spawn Probability=true
    UseLoopCountLimit=false
   )
  ]
  particleSpawnScript [
   &PartitionParticles()
   &InitializeParticle(
    Color=&LerpLinearColors(StartColor=<rgba 0 1 0.2333 2>, EndColor=<rgba 0.5 0 1 2>, LerpFactor=$User.VariantSwitch)
    Lifetime=0.4
    Sprite Size Mode=::NewEnumerator3
    Uniform Sprite Size=400
    UsePositionOffset=false
    Write Material Random=true
   )
   &DynamicMaterialParameters(
    Index 0 Param 1=30
    Index 0 Param 2=30
    Module.Index0Param0Write=true
    Module.Index0Param1Write=true
    Module.Index0Param2Write=true
    Module.Index0Param3Write=true
   )
   &InitialMeshOrientation(Mesh Orientation Mode=::NewEnumerator2, Rotation=<-0.3 0 0>, Use Rotation Vector=true)
   &CameraOffset(Camera Offset Amount=50)
  ]
  particleUpdateScript [
   &ParticleState()
   &SkeletalMeshLocation(
    Bone Sampling Mode=::NewEnumerator4
    MaskSpawnGroup=false
    Sampled Position Offset=&SelectVectorFromArray(
     Direct Array Index=$Particles.PartitionParticles.Partition
     Vector Selection Array=DataInterface() {propertyValues {FloatData [<-75 -23 -11>, <75 -23 -11>]}}
    )
    Skeletal Mesh=$System.Skeletal Mesh
   )
   &ScaleColor(
    Scale Alpha=&FloatFromCurve(FloatCurve=DataInterface() {propertyValues {Curve {keys [{interpMode=RCIM_Cubic, tangentMode=RCTM_User, time=0, value=1}, {interpMode=RCIM_Cubic, tangentMode=RCTM_User, time=1, value=0}]}}})
    Scale RGB=&Multiply_VectorByFloat(
     Vector=<1 1 1>
     Float=&FloatFromCurve(FloatCurve=DataInterface() {propertyValues {Curve {keys [
          {interpMode=RCIM_Cubic, tangentMode=RCTM_User, time=0, value=2}
          <at 0.7 1.5>
          {arriveTangent=-4.4348, interpMode=RCIM_Cubic, leaveTangent=-4.4348, tangentMode=RCTM_User, time=1, value=0}
         ]}}})
    )
    ScaleA=true
    ScaleRGB=true
   )
  ]
  rendererClasses [&NiagaraSpriteRendererProperties]
  renderers [{rendererClass=&NiagaraSpriteRendererProperties}]
 }
 Flash001 {
  emitterUpdateScript [
   &EmitterState(Delay First Loop Only=true, Life Cycle Mode=::NewEnumerator1, Loop Delay=$System.Delay, Loop Duration=1, UseLoopDelay=true)
   &SpawnPerFrame(Spawn:NiagaraBool=$System.DoOnce.Execute, Spawn Count=2, Use Spawn Probability=false)
  ]
  particleSpawnScript [
   &PartitionParticles()
   &InitializeParticle(
    Color=&LerpLinearColors(StartColor=<rgba 0 1 0.2333 0.5>, EndColor=<rgba 0.5 0 1 0.5>, LerpFactor=$User.VariantSwitch)
    Lifetime=0.4
    Sprite Size Mode=::NewEnumerator3
    Uniform Sprite Size=500
    UsePositionOffset=false
    Write Material Random=true
   )
   &DynamicMaterialParameters(
    Index 0 Param 1=30
    Index 0 Param 2=30
    Module.Index0Param0Write=true
    Module.Index0Param1Write=true
    Module.Index0Param2Write=true
    Module.Index0Param3Write=true
   )
   &InitialMeshOrientation(Mesh Orientation Mode=::NewEnumerator2, Rotation=<-0.3 0 0>, Use Rotation Vector=true)
  ]
  particleUpdateScript [
   &ParticleState()
   &SkeletalMeshLocation(
    Bone Sampling Mode=::NewEnumerator4
    MaskSpawnGroup=false
    Sampled Position Offset=&SelectVectorFromArray(
     Direct Array Index=$Particles.PartitionParticles.Partition
     Vector Selection Array=DataInterface() {propertyValues {FloatData [<-75 -23 -11>, <75 -23 -11>]}}
    )
    Skeletal Mesh=$System.Skeletal Mesh
   )
   &ScaleColor(
    Scale Alpha=&FloatFromCurve(FloatCurve=DataInterface() {propertyValues {Curve {keys [{interpMode=RCIM_Cubic, tangentMode=RCTM_User, time=0, value=1}, {interpMode=RCIM_Cubic, tangentMode=RCTM_User, time=1, value=0}]}}})
    Scale RGB=&Multiply_VectorByFloat(
     Vector=<1 1 1>
     Float=&FloatFromCurve(FloatCurve=DataInterface() {propertyValues {Curve {keys [
          {interpMode=RCIM_Cubic, tangentMode=RCTM_User, time=0, value=2}
          <at 0.7 1.5>
          {arriveTangent=-4.4348, interpMode=RCIM_Cubic, leaveTangent=-4.4348, tangentMode=RCTM_User, time=1, value=0}
         ]}}})
    )
    ScaleA=true
    ScaleRGB=true
   )
  ]
  rendererClasses [&NiagaraSpriteRendererProperties]
  renderers [{rendererClass=&NiagaraSpriteRendererProperties}]
 }
 Green_Smoke {
  emitterUpdateScript [
   &EmitterState(
    Inactive Response=::NewEnumerator1
    Life Cycle Mode=::NewEnumerator1
    Loop Behavior=::NewEnumerator1
    Loop Delay=$System.Delay
    Loop Duration Mode=::NewEnumerator1
    UseLoopDelay=true
   )
   &SpawnBurst_Instantaneous(
    Spawn Count=4
    Spawn Probability=&MakeFloatFromInt(Int=&MakeCustomIntFromBool(Bool=$User.bIsFrontEndPreview, True INT=0, False INT=1))
    Use Spawn Probability=true
    UseLoopCountLimit=false
   )
  ]
  particleSpawnScript [
   &PartitionParticles(Number Of Partitions=4)
   &InitializeParticle(
    Color=&LerpLinearColors(StartColor=<rgba 0 8 1.948 1.2>, EndColor=<rgba 8.25 1.5 15 1.2>, LerpFactor=$User.VariantSwitch)
    Mesh Scale=&SelectVectorFromArray(
     Direct Array Index=$Particles.PartitionParticles.Partition
     Vector Selection Array=DataInterface() {propertyValues {FloatData [<1 1 1.5>, <2 1 1.5>, <1 1 1.5>, <1.5 1 1.5>]}}
    )
    Mesh Scale Mode=::NewEnumerator4
    UsePositionOffset=false
    Write Material Random=true
   )
   &InitialMeshOrientation(Mesh Orientation Mode=::NewEnumerator2, Rotation=<0 0 0.25>, Use Rotation Vector=true)
   &InitialMeshOrientation(
    Mesh Orientation Mode=::NewEnumerator2
    Rotation=&SelectVectorFromArray(
     Direct Array Index=$Particles.PartitionParticles.Partition
     Vector Selection Array=DataInterface() {propertyValues {FloatData [<-0.06 0.1 0>, <0.05 0.1 0>, <-0.05 -0.1 0.05>, <0.03 -0.1 0>]}}
    )
    Use Rotation Vector=true
   ) {module=InitialMeshOrientation001}
   &DynamicMaterialParameters(
    Index 0 Param 1=100
    Index 0 Param 2=70
    Index 0 Param 3=&SelectFloatFromArray(Direct Array Index=$Particles.PartitionParticles.Partition, Float Selection Array=DataInterface() {propertyValues {FloatData [1,1.5,1,1.5]}})
    Index 0 Param 4=&SelectFloatFromArray(Direct Array Index=$Particles.PartitionParticles.Partition, Float Selection Array=DataInterface() {propertyValues {FloatData [0,0,0,0.4]}})
    Module.Index0Param0Write=true
    Module.Index0Param1Write=true
    Module.Index0Param2Write=true
    Module.Index0Param3Write=true
   )
   &CameraOffset(Camera Offset Amount=10)
  ]
  particleUpdateScript [
   &ParticleState(Kill Particles When Lifetime Has Elapsed=false, Let Infinitely Lived Particles Die When Emitter Deactivates=true)
   &ScaleMeshSize(Scale Factor=&Lerp_Vector(A=<0.5 1 1>, B=<1 1 1>, Alpha=$System.InterpolateOverTime.InterpolateOverTime_Float.Moving Average))
   &SkeletalMeshLocation(
    Bone Sampling Mode=::NewEnumerator4
    MaskSpawnGroup=false
    Sampled Position Offset=&SelectVectorFromArray(
     Direct Array Index=$Particles.PartitionParticles.Partition
     Vector Selection Array=DataInterface() {propertyValues {FloatData [<70 0 -20>, <65 0 -10>, <-70 0 -20>, <-65 0 -10>]}}
    )
    Skeletal Mesh=$System.Skeletal Mesh
   )
   &InterpolateOverTime(Rate Of Change:NiagaraFloat=$System.Rate of Change, Target Vector Value:NiagaraPosition=$Particles.Position, Variable Type=::NewEnumerator2) {enabled=false}
   /SetVariables(Particles.InterpolateDy:Vector4f=&MakeFloat4FromVectorAndFloat("Vector (RGB)"=$Particles.InterpolateOverTime.InterpolateOverTime_V3_.Moving Average, "Float (Alpha)"=$System.Interpolate WPO)) {enabled=false, isSetParameters=true, module=SetVariables_#8}
   &ScaleColor(
    Scale Alpha=&Multiply_Float(
     A=&Lerp_Float(A=0.5, B=1.2, Alpha=$System.InterpolateOverTime.InterpolateOverTime_Float.Moving Average)
     B=&MakeCustomFloatFromBool(
      Bool=&SetBoolByIntComparison(A=$Particles.PartitionParticles.Partition, B=2, Comparison Type=::NewEnumerator0)
      True Float=&Lerp_Float(A=1, B=0.3, Alpha=$System.InterpolateOverTime004.InterpolateOverTime_Float.Moving Average)
      False Float=&Lerp_Float(A=1, B=0.3, Alpha=$System.InterpolateOverTime005.InterpolateOverTime_Float.Moving Average)
     )
    )
    ScaleA=true
    ScaleRGB=true
   )
   /SetVariables(Particles.MeshOrientation=&MultiplyQuaternion(
     Quaternion A=$Particles.MeshOrientation
     Quaternion B=&MultiplyQuaternion(
      Quaternion A=&MultiplyQuaternion(
       Quaternion A=&MakeQuatFromAxisAngle(
        Axis=<0 1 0>
        AngleInDegrees=&FloatFromCurve(
         FloatCurve=DataInterface() {propertyValues {Curve {keys [
             {interpMode=RCIM_Cubic, tangentMode=RCTM_Break, time=0, value=0}
             {interpMode=RCIM_Cubic, tangentMode=RCTM_Break, time=0.5, value=1}
             {interpMode=RCIM_Cubic, tangentMode=RCTM_Break, time=0.9, value=0}
            ]}}}
         CurveIndex=$System.InterpolateOverTime001.InterpolateOverTime_Float.Moving Average
         Scale Curve=-30
        )
       )
       Quaternion B=&MultiplyQuaternion(
        Quaternion A=&MakeQuatFromAxisAngle(
         Axis=<1 0 0>
         AngleInDegrees=&FloatFromCurve(
          FloatCurve=DataInterface() {propertyValues {Curve {keys [
              {arriveTangent=5.3968, interpMode=RCIM_Cubic, leaveTangent=5.3968, tangentMode=RCTM_User, time=0, value=0}
              {interpMode=RCIM_Cubic, tangentMode=RCTM_Break, time=0.5, value=1}
              {interpMode=RCIM_Cubic, tangentMode=RCTM_Break, time=0.9, value=0}
             ]}}}
          CurveIndex=$System.InterpolateOverTime002.InterpolateOverTime_Float.Moving Average
          Scale Curve=-30
         )
        )
        Quaternion B=&MakeQuatFromAxisAngle(
         Axis=<1 0 0>
         AngleInDegrees=&FloatFromCurve(
          FloatCurve=DataInterface() {propertyValues {Curve {keys [
              {arriveTangent=5.3968, interpMode=RCIM_Cubic, leaveTangent=5.3968, tangentMode=RCTM_User, time=0, value=0}
              {interpMode=RCIM_Cubic, tangentMode=RCTM_Break, time=0.5, value=1}
              {interpMode=RCIM_Cubic, tangentMode=RCTM_Break, time=0.9, value=0}
             ]}}}
          CurveIndex=$System.InterpolateOverTime003.InterpolateOverTime_Float.Moving Average
          Scale Curve=30
         )
        )
       )
      )
      Quaternion B=&MakeQuatFromAxisAngle(
       Axis=<0 1 0>
       AngleInDegrees=&FloatFromCurve(
        FloatCurve=DataInterface() {propertyValues {Curve {keys [
            {interpMode=RCIM_Cubic, tangentMode=RCTM_Break, time=2, value=0}
            {interpMode=RCIM_Cubic, tangentMode=RCTM_Break, time=3, value=1}
            {interpMode=RCIM_Cubic, tangentMode=RCTM_Break, time=4, value=0}
           ]}}}
        CurveIndex=$Emitter.Age
        Scale Curve=-30
       )
      )
     )
    )) {isSetParameters=true}
  ]
  rendererClasses [&NiagaraMeshRendererProperties]
  renderers [{rendererClass=&NiagaraMeshRendererProperties, rendererIndex=0}]
 }
 Light {
  emitterUpdateScript [
   &EmitterState(
    Inactive Response=::NewEnumerator1
    Life Cycle Mode=::NewEnumerator1
    Loop Behavior=::NewEnumerator1
    Loop Delay=$System.Delay
    Loop Duration Mode=::NewEnumerator1
    UseLoopDelay=true
   )
   &SpawnBurst_Instantaneous(Use Spawn Probability=false, UseLoopCountLimit=false)
  ]
  particleSpawnScript [&InitializeParticle(
    Color Mode=::NewEnumerator0
    Mesh Scale=&Multiply_VectorByFloat(
     Vector=&SelectVectorFromArray(
      Direct Array Index=$Particles.PartitionParticles.Partition
      Vector Selection Array=DataInterface() {propertyValues {FloatData [<1 1 1>, <1 1 1>]}}
     )
     Float=1
    )
    Mesh Scale Mode=::NewEnumerator4
    Sprite Size Mode=::NewEnumerator3
    UsePositionOffset=false
    Write Material Random=true
   )]
  particleUpdateScript [
   &ParticleState(Kill Particles When Lifetime Has Elapsed=false, Let Infinitely Lived Particles Die When Emitter Deactivates=true)
   &SkeletalMeshLocation(Bone Sampling Mode=::NewEnumerator4, MaskSpawnGroup=false, Sampled Position Offset=<0 40 0>, Skeletal Mesh=$System.Skeletal Mesh)
   &FN_Light_Attributes(
    Color=&LerpLinearColors(StartColor=<rgba 0 30 5 1>, EndColor=<rgba 15 0 30 1>, LerpFactor=$User.VariantSwitch)
    Create Color Attribute=true
    Create Exponent Attribute=true
    Create Radius Attribute=true
    Exponent=3
    Radius=100
    Scale Color=&Lerp_Float(A=0, B=1, Alpha=$System.InterpolateOverTime.InterpolateOverTime_Float.Moving Average)
   )
  ]
  rendererClasses [&NiagaraLightRendererProperties]
  renderers [{rendererClass=&NiagaraLightRendererProperties, rendererIndex=0}]
 }
 Skull_Sprite {
  emitterUpdateScript [
   &EmitterState(
    Delay First Loop Only=true
    Inactive Response=::NewEnumerator1
    Life Cycle Mode=::NewEnumerator1
    Loop Delay=&Add_Float(A=$System.Delay, B=0.5)
    Loop Duration=1
    UseLoopDelay=true
   )
   &SpawnRate(
    SpawnRate=&Multiply_Float(A=6, B=&Lerp_Float(A=0, B=1, Alpha=$System.InterpolateOverTime.InterpolateOverTime_Float.Moving Average))
    Use Spawn Probability=false
   )
  ]
  particleSpawnScript [
   &PartitionParticles()
   &InitializeParticle(
    Color=&LerpLinearColors(StartColor=<rgba 1 1 1 1>, EndColor=<rgba 1.3 0.6 2 1>, LerpFactor=$User.VariantSwitch)
    Lifetime Min=1.5
    Lifetime Mode=::NewEnumerator1
    Mesh Scale=&Multiply_VectorByFloat(Vector=<1 1 1>, Float=&RandomRangeFloat(Minimum=3.5, Maximum=3))
    Mesh Scale Mode=::NewEnumerator4
    UsePositionOffset=false
    Write Material Random=true
   )
   &SkeletalMeshLocation(
    Bone Sampling Mode=::NewEnumerator4
    MaskSpawnGroup=false
    Orientation Sampling=::NewEnumerator1
    Sampled Position Offset=&AddVectorToPosition(
     Position=&SelectVectorFromArray(
      Direct Array Index=$Particles.PartitionParticles.Partition
      Vector Selection Array=DataInterface() {propertyValues {FloatData [<-75 -23 -11>, <75 -23 -11>]}}
     )
     Vector=<0 0 0>
    )
    Skeletal Mesh=$System.Skeletal Mesh
   )
   &ShapeLocation(MaskSpawnGroup=false, Override Seed=false, Sphere Radius=30, UseOwnerScale=true)
   &AddVelocity(
    Cone Angle=1
    Cone Axis=&SelectVectorFromArray(
     Direct Array Index=$Particles.PartitionParticles.Partition
     Vector Selection Array=DataInterface() {propertyValues {FloatData [<0.7 -0.7 0.7>, <-0.7 -0.7 0.7>]}}
    )
    Module.UseConeFalloff=false
    Module.UseVelDistribution=false
    Override Seed=false
    Velocity Mode=::NewEnumerator2
    Velocity Speed=300
   )
   &InitialMeshOrientation(
    Mesh Orientation Mode=::NewEnumerator2
    Rotation=&SelectVectorFromArray(
     Direct Array Index=$Particles.PartitionParticles.Partition
     Vector Selection Array=DataInterface() {propertyValues {FloatData [<0 0 0.5>, <0.07 0.25 0>]}}
    )
    Use Rotation Vector=true
   ) {enabled=false}
   &InitialMeshOrientation(
    Mesh Orientation Mode=::NewEnumerator2
    Rotation=&SelectVectorFromArray(
     Direct Array Index=$Particles.PartitionParticles.Partition
     Vector Selection Array=DataInterface() {propertyValues {FloatData [<-0.07 0.25 0>, <0 0 0>]}}
    )
    Use Rotation Vector=true
   ) {enabled=false, module=InitialMeshOrientation001}
   &InitialMeshOrientation(
    Mesh Axis To Orient=<0 1 0>
    Mesh Orientation Mode=::NewEnumerator2
    Rotation=&RandomRangeVector(Minimum=<-0.1 0 -0.035>, Maximum=<0 0 0.035>)
    Use Rotation Vector=true
   ) {module=InitialMeshOrientation002}
  ]
  particleUpdateScript [
   &ParticleState()
   &ScaleMeshSize(Scale Factor=&Multiply_VectorByFloat(
     Vector=<1 1 1>
     Float=&FloatFromCurve(
      FloatCurve=DataInterface() {propertyValues {Curve {keys [
          {interpMode=RCIM_Cubic, tangentMode=RCTM_User, time=0, value=0}
          <at 0.2 1>
          <at 0.6 1.075>
          {interpMode=RCIM_Cubic, tangentMode=RCTM_User, time=1, value=0}
         ]}}}
      Scale Curve=0.5
     )
    ))
   &CurlNoiseForce(Noise Frequency=10, Noise Strength=300, Pan Noise=false) {enabled=false}
   &Drag(Use Linear Drag=true, Use Rotational Drag=true)
   &SolveForcesAndVelocity(Clamp Velocity=false, Limit Acceleration=false)
   &ScaleColor(
    Scale Alpha=&FloatFromCurve(
     FloatCurve=DataInterface() {propertyValues {Curve {keys [
         {arriveTangent=1.7241, interpMode=RCIM_Cubic, leaveTangent=1.7241, tangentMode=RCTM_User, time=0, value=0}
         {arriveTangent=6.2337, interpMode=RCIM_Cubic, leaveTangent=-1.8543, tangentMode=RCTM_Break, time=0.2, value=1}
         {tangentMode=RCTM_User, time=1, value=0}
        ]}}}
     Scale Curve=&Multiply_Float(
      A=&Lerp_Float(A=0, B=1, Alpha=$System.InterpolateOverTime.InterpolateOverTime_Float.Moving Average)
      B=&MakeCustomFloatFromBool(
       Bool=&SetBoolByIntComparison(A=$Particles.PartitionParticles.Partition, B=0, Comparison Type=::NewEnumerator0)
       True Float=&Lerp_Float(A=1, B=0, Alpha=$System.InterpolateOverTime005.InterpolateOverTime_Float.Moving Average)
       False Float=&Lerp_Float(A=1, B=0, Alpha=$System.InterpolateOverTime004.InterpolateOverTime_Float.Moving Average)
      )
     )
    )
    Scale RGB=&MakeVectorFromLinearColorRGB(Color=&LerpLinearColors(
      StartColor=<rgba 1 1 1 1>
      EndColor=&FNTimeOfDayColor(Night Color=<rgba 2 0.6458 1.3543 1>, Morning Color=<rgba 1 1 1 1>, Day Color=<rgba 1 1 1 1>, Evening Color=<rgba 0.6738 1.0335 2 1>)
      LerpFactor=$User.VariantSwitch
     ))
    ScaleA=true
    ScaleRGB=true
   )
   &FN_Sprite_Attributes(
    Color=&MakeLinearColorFromVectorAndFloat(
     "Vector (RGB)"=&MakeVectorFromLinearColorRGB(Color=&LerpLinearColors(StartColor=<rgba 0 1 0.3053 1>, EndColor=<rgba 1.1 0.2 2 1>, LerpFactor=$User.VariantSwitch))
     "Float (Alpha)"=&FloatFromCurve(
      FloatCurve=DataInterface() {propertyValues {Curve {keys [
          {interpMode=RCIM_Cubic, tangentMode=RCTM_User, time=0, value=0}
          <at 0.1 0.875>
          <at 0.5 0.8>
          {interpMode=RCIM_Cubic, tangentMode=RCTM_User, time=1, value=0}
         ]}}}
      Scale Curve=&Multiply_Float(
       A=&Multiply_Float(
        A=&Multiply_Float(A=0.2, B=&FN_PickFloatDependOnPlatform(TargetPlatformActive=1, TargetPlatformNoActive=0.7))
        B=&MakeCustomFloatFromBool(
         Bool=&SetBoolByIntComparison(A=$Particles.PartitionParticles.Partition, B=0, Comparison Type=::NewEnumerator0)
         True Float=&Lerp_Float(A=1, B=0, Alpha=$System.InterpolateOverTime005.InterpolateOverTime_Float.Moving Average)
         False Float=&Lerp_Float(A=1, B=0, Alpha=$System.InterpolateOverTime004.InterpolateOverTime_Float.Moving Average)
        )
       )
       B=&Lerp_Float(A=0, B=1, Alpha=$System.InterpolateOverTime.InterpolateOverTime_Float.Moving Average)
      )
     )
    )
    Create Color Attribute=true
    Create SpriteSize Attribute=true
    SpriteSize=&Vector2DFromFloat(Value=70)
    Write to Module Namespace=true
   )
  ]
  rendererClasses [&NiagaraMeshRendererProperties, &NiagaraSpriteRendererProperties]
  renderers [
   {rendererClass=&NiagaraMeshRendererProperties, rendererIndex=0}
   {rendererClass=&NiagaraSpriteRendererProperties, rendererIndex=1}
  ]
 }
 Smoke_Dark {
  emitterUpdateScript [
   &EmitterState(
    Inactive Response=::NewEnumerator1
    Life Cycle Mode=::NewEnumerator1
    Loop Behavior=::NewEnumerator1
    Loop Delay=$System.Delay
    Loop Duration Mode=::NewEnumerator1
    UseLoopDelay=true
   )
   &SpawnBurst_Instantaneous(
    Spawn Count=6
    Spawn Probability=&MakeFloatFromInt(Int=&MakeCustomIntFromBool(Bool=$User.bIsFrontEndPreview, True INT=0, False INT=1))
    Use Spawn Probability=true
    UseLoopCountLimit=false
   )
  ]
  particleSpawnScript [
   &PartitionParticles(Number Of Partitions=6)
   &InitializeParticle(
    Color=&LerpLinearColors(StartColor=<rgba 0.01 0.01 0.01 1>, EndColor=<rgba 0.018 0.012 0.03 1>, LerpFactor=$User.VariantSwitch)
    Mesh Scale=&SelectVectorFromArray(
     Direct Array Index=$Particles.PartitionParticles.Partition
     Vector Selection Array=DataInterface() {propertyValues {FloatData [<1 1 1.5>, <1 1 1>, <2 1 2>, <1 1 1.5>, <1 1 1>, <2 1 2>]}}
    )
    Mesh Scale Mode=::NewEnumerator4
    UsePositionOffset=false
    Write Material Random=true
   )
   &InitialMeshOrientation(Mesh Orientation Mode=::NewEnumerator2, Rotation=<0 0 0.25>, Use Rotation Vector=true)
   &InitialMeshOrientation(
    Mesh Orientation Mode=::NewEnumerator2
    Rotation=&SelectVectorFromArray(
     Direct Array Index=$Particles.PartitionParticles.Partition
     Vector Selection Array=DataInterface() {propertyValues {FloatData [<-0.08 0.1 -0.05>, <0 0.1 -0.05>, <0.03 0.1 0>, <-0.03 -0.1 0.07>, <-0.07 -0.1 0.07>, <0.01 -0.03 0.07>]}}
    )
    Use Rotation Vector=true
   ) {module=InitialMeshOrientation001}
   &CameraOffset(Camera Offset Amount=10)
  ]
  particleUpdateScript [
   &DynamicMaterialParameters(
    Index 0 Param 1=&SelectFloatFromArray(
     Direct Array Index=$Particles.PartitionParticles.Partition
     Float Selection Array=DataInterface() {propertyValues {FloatData [150,150,200,150,150,200]}}
    )
    Index 0 Param 2=&SelectFloatFromArray(
     Direct Array Index=$Particles.PartitionParticles.Partition
     Float Selection Array=DataInterface() {propertyValues {FloatData [60,70,50,60,70,50]}}
    )
    Index 0 Param 3=&SelectFloatFromArray(Direct Array Index=$Particles.PartitionParticles.Partition, Float Selection Array=DataInterface() {propertyValues {FloatData [1,1,1.5]}})
    Index 0 Param 4=&SelectFloatFromArray(
     Direct Array Index=$Particles.PartitionParticles.Partition
     Float Selection Array=DataInterface() {propertyValues {FloatData [0,0.4,0,0,0.3,0]}}
    )
    Index 1 Param 1=&Lerp_Float(A=1, B=0, Alpha=$System.InterpolateOverTime.InterpolateOverTime_Float.Moving Average)
    Index 1 Param 2=&FloatFromCurve(FloatCurve=DataInterface() {propertyValues {Curve {keys [{tangentMode=RCTM_User, time=0, value=0}, {tangentMode=RCTM_User, time=0.5, value=1}]}}})
    Index 1 Param 3=&MakeCustomFloatFromBool(
     Bool=&SetBoolByIntComparison(A=$Particles.PartitionParticles.Partition, B=2, Comparison Type=::NewEnumerator0)
     True Float=&Lerp_Float(A=0, B=3, Alpha=$System.InterpolateOverTime004.InterpolateOverTime_Float.Moving Average)
     False Float=&Lerp_Float(A=0, B=3, Alpha=$System.InterpolateOverTime005.InterpolateOverTime_Float.Moving Average)
    )
    Module.Index0Param0Write=true
    Module.Index0Param1Write=true
    Module.Index0Param2Write=true
    Module.Index0Param3Write=true
    Module.Index1Param0Write=true
    Module.Index1Param1Write=true
    Module.Index1Param2Write=true
    Module.Index1Param3Write=true
    Write Parameter Index 1=true
   )
   &ParticleState(Kill Particles When Lifetime Has Elapsed=false, Let Infinitely Lived Particles Die When Emitter Deactivates=true)
   &ScaleMeshSize(Scale Factor=&Lerp_Vector(A=<0.5 1 1>, B=<1 1 1>, Alpha=$System.InterpolateOverTime.InterpolateOverTime_Float.Moving Average))
   &ScaleMeshSize(Scale Factor=&MakeCustomVectorFromBool(
     Bool=&SetBoolByIntComparison(A=$Particles.PartitionParticles.Partition, B=2, Comparison Type=::NewEnumerator0)
     True Vector=&Lerp_Vector(A=<1 1 1>, B=<0.8 0.8 0.8>, Alpha=$System.InterpolateOverTime004.InterpolateOverTime_Float.Moving Average)
     False Vector=&Lerp_Vector(A=<1 1 1>, B=<0.8 0.8 0.8>, Alpha=$System.InterpolateOverTime005.InterpolateOverTime_Float.Moving Average)
    )) {module=ScaleMeshSize001}
   &SkeletalMeshLocation(
    Bone Sampling Mode=::NewEnumerator4
    MaskSpawnGroup=false
    Sampled Position Offset=&SelectVectorFromArray(
     Direct Array Index=$Particles.PartitionParticles.Partition
     Vector Selection Array=DataInterface() {propertyValues {FloatData [<70 0 -10>, <70 0 -10>, <65 0 0>, <-70 -5 0>, <-70 -5 10>, <-60 -5 10>]}}
    )
    Skeletal Mesh=$System.Skeletal Mesh
   )
   &InterpolateOverTime(Rate Of Change:NiagaraFloat=$System.Rate of Change, Target Vector Value:NiagaraPosition=$Particles.Position, Variable Type=::NewEnumerator2) {enabled=false}
   /SetVariables(Particles.InterpolateDy:Vector4f=&MakeFloat4FromVectorAndFloat("Vector (RGB)"=$Particles.InterpolateOverTime.InterpolateOverTime_V3_.Moving Average, "Float (Alpha)"=$System.Interpolate WPO)) {enabled=false, isSetParameters=true, module=SetVariables_#8}
   /SetVariables(Particles.MeshOrientation=&MultiplyQuaternion(
     Quaternion A=$Particles.MeshOrientation
     Quaternion B=&MultiplyQuaternion(
      Quaternion A=&MultiplyQuaternion(
       Quaternion A=&MakeQuatFromAxisAngle(
        Axis=<0 1 0>
        AngleInDegrees=&FloatFromCurve(
         FloatCurve=DataInterface() {propertyValues {Curve {keys [
             {interpMode=RCIM_Cubic, tangentMode=RCTM_Break, time=0, value=0}
             {interpMode=RCIM_Cubic, tangentMode=RCTM_Break, time=0.5, value=1}
             {interpMode=RCIM_Cubic, tangentMode=RCTM_Break, time=0.9, value=0}
            ]}}}
         CurveIndex=$System.InterpolateOverTime001.InterpolateOverTime_Float.Moving Average
         Scale Curve=-30
        )
       )
       Quaternion B=&MultiplyQuaternion(
        Quaternion A=&MakeQuatFromAxisAngle(
         Axis=<1 0 0>
         AngleInDegrees=&FloatFromCurve(
          FloatCurve=DataInterface() {propertyValues {Curve {keys [
              {arriveTangent=5.3968, interpMode=RCIM_Cubic, leaveTangent=5.3968, tangentMode=RCTM_User, time=0, value=0}
              {interpMode=RCIM_Cubic, tangentMode=RCTM_Break, time=0.5, value=1}
              {interpMode=RCIM_Cubic, tangentMode=RCTM_Break, time=0.9, value=0}
             ]}}}
          CurveIndex=$System.InterpolateOverTime002.InterpolateOverTime_Float.Moving Average
          Scale Curve=-30
         )
        )
        Quaternion B=&MakeQuatFromAxisAngle(
         Axis=<1 0 0>
         AngleInDegrees=&FloatFromCurve(
          FloatCurve=DataInterface() {propertyValues {Curve {keys [
              {arriveTangent=5.3968, interpMode=RCIM_Cubic, leaveTangent=5.3968, tangentMode=RCTM_User, time=0, value=0}
              {interpMode=RCIM_Cubic, tangentMode=RCTM_Break, time=0.5, value=1}
              {interpMode=RCIM_Cubic, tangentMode=RCTM_Break, time=0.9, value=0}
             ]}}}
          CurveIndex=$System.InterpolateOverTime003.InterpolateOverTime_Float.Moving Average
          Scale Curve=30
         )
        )
       )
      )
      Quaternion B=&MakeQuatFromAxisAngle(
       Axis=<0 1 0>
       AngleInDegrees=&FloatFromCurve(
        FloatCurve=DataInterface() {propertyValues {Curve {keys [
            {interpMode=RCIM_Cubic, tangentMode=RCTM_Break, time=2, value=0}
            {interpMode=RCIM_Cubic, tangentMode=RCTM_Break, time=3, value=1}
            {interpMode=RCIM_Cubic, tangentMode=RCTM_Break, time=4, value=0}
           ]}}}
        CurveIndex=$Emitter.Age
        Scale Curve=-30
       )
      )
     )
    )) {isSetParameters=true, module=SetVariables_#9}
  ]
  rendererClasses [&NiagaraMeshRendererProperties]
  renderers [{rendererClass=&NiagaraMeshRendererProperties, rendererIndex=0}]
 }
 Smoke_Dark_Flash {
  emitterUpdateScript [
   &EmitterState(Life Cycle Mode=::NewEnumerator1, Loop Behavior=::NewEnumerator1, Loop Delay=$System.Delay, Loop Duration=1, UseLoopDelay=true)
   &SpawnBurst_Instantaneous(
    Spawn Count=6
    Spawn Probability=&MakeFloatFromInt(Int=&MakeCustomIntFromBool(Bool=$User.bIsFrontEndPreview, True INT=0, False INT=1))
    Use Spawn Probability=true
    UseLoopCountLimit=false
   )
  ]
  particleSpawnScript [
   &PartitionParticles(Number Of Partitions=6)
   &InitializeParticle(
    Color=<rgba 1.5 1.5 1.5 0.7>
    Lifetime=0.5
    Mesh Scale=&SelectVectorFromArray(
     Direct Array Index=$Particles.PartitionParticles.Partition
     Vector Selection Array=DataInterface() {propertyValues {FloatData [<1 1 1.5>, <1 1 1>, <2 1 2>, <1 1 1.5>, <1 1 1>, <2 1 2>]}}
    )
    Mesh Scale Mode=::NewEnumerator4
    UsePositionOffset=false
    Write Material Random=true
   )
   &InitialMeshOrientation(Mesh Orientation Mode=::NewEnumerator2, Rotation=<0 0 0.25>, Use Rotation Vector=true)
   &InitialMeshOrientation(
    Mesh Orientation Mode=::NewEnumerator2
    Rotation=&SelectVectorFromArray(
     Direct Array Index=$Particles.PartitionParticles.Partition
     Vector Selection Array=DataInterface() {propertyValues {FloatData [<-0.08 0.1 -0.05>, <0 0.1 -0.05>, <0.03 0.1 0>, <-0.03 -0.1 0.07>, <-0.07 -0.1 0.07>, <0.01 -0.03 0.07>]}}
    )
    Use Rotation Vector=true
   ) {module=InitialMeshOrientation001}
   &CameraOffset(Camera Offset Amount=10)
  ]
  particleUpdateScript [
   &DynamicMaterialParameters(
    Index 0 Param 1=&SelectFloatFromArray(
     Direct Array Index=$Particles.PartitionParticles.Partition
     Float Selection Array=DataInterface() {propertyValues {FloatData [150,150,200,150,150,200]}}
    )
    Index 0 Param 2=&SelectFloatFromArray(
     Direct Array Index=$Particles.PartitionParticles.Partition
     Float Selection Array=DataInterface() {propertyValues {FloatData [60,70,50,60,70,50]}}
    )
    Index 0 Param 3=&SelectFloatFromArray(Direct Array Index=$Particles.PartitionParticles.Partition, Float Selection Array=DataInterface() {propertyValues {FloatData [1,1,1.5]}})
    Index 0 Param 4=&SelectFloatFromArray(
     Direct Array Index=$Particles.PartitionParticles.Partition
     Float Selection Array=DataInterface() {propertyValues {FloatData [0,0.4,0,0,0.3,0]}}
    )
    Index 1 Param 1=0.5
    Index 1 Param 2=&FloatFromCurve(FloatCurve=DataInterface() {propertyValues {Curve {keys [{tangentMode=RCTM_User, time=0, value=0}, {tangentMode=RCTM_User, time=0.5, value=1}]}}})
    Module.Index0Param0Write=true
    Module.Index0Param1Write=true
    Module.Index0Param2Write=true
    Module.Index0Param3Write=true
    Module.Index1Param0Write=true
    Module.Index1Param1Write=true
    Module.Index1Param2Write=true
    Module.Index1Param3Write=true
    Write Parameter Index 1=true
   )
   &ParticleState()
   &SkeletalMeshLocation(
    Bone Sampling Mode=::NewEnumerator4
    MaskSpawnGroup=false
    Sampled Position Offset=&SelectVectorFromArray(
     Direct Array Index=$Particles.PartitionParticles.Partition
     Vector Selection Array=DataInterface() {propertyValues {FloatData [<70 0 -10>, <70 0 -10>, <65 0 0>, <-70 -5 0>, <-70 -5 10>, <-60 -5 10>]}}
    )
    Skeletal Mesh=$System.Skeletal Mesh
   )
   &ScaleColor(
    Scale Alpha=&FloatFromCurve(FloatCurve=DataInterface() {propertyValues {Curve {keys [<at 0 0>, {interpMode=RCIM_Cubic, leaveTangent=-1.732, tangentMode=RCTM_Break, time=0.2, value=0.8}, <at 1 0>]}}})
    Scale RGB=&MakeVectorFromLinearColorRGB(Color=&LerpLinearColors(
      StartColor=&ColorFromCurve(ColorCurve=DataInterface() {propertyValues {
         AlphaCurve {keys []}
         BlueCurve {keys [<at 0 0.5 RCIM_Cubic>,<at 0.8319 0 RCIM_Cubic>]}
         GreenCurve {keys [<at 0 3 RCIM_Cubic>,<at 0.8319 1 RCIM_Cubic>]}
         RedCurve {keys [<at 0 0 RCIM_Cubic>,<at 0.8319 0.1667 RCIM_Cubic>]}
        }})
      EndColor=&ColorFromCurve(ColorCurve=DataInterface() {propertyValues {
         AlphaCurve {keys []}
         BlueCurve {keys [<at 0 3 RCIM_Cubic>,<at 0.6283 0.776 RCIM_Cubic>]}
         GreenCurve {keys [<at 0 0.9062 RCIM_Cubic>,<at 0.6283 0 RCIM_Cubic>]}
         RedCurve {keys [<at 0 1.9531 RCIM_Cubic>,<at 0.6283 0.388 RCIM_Cubic>]}
        }})
      LerpFactor=$User.VariantSwitch
     ))
    ScaleA=true
    ScaleRGB=true
   )
  ]
  rendererClasses [&NiagaraMeshRendererProperties]
  renderers [{rendererClass=&NiagaraMeshRendererProperties, rendererIndex=0}]
 }
 Smoke_Sprite_dark {
  emitterUpdateScript [
   &EmitterState(
    Delay First Loop Only=true
    Inactive Response=::NewEnumerator1
    Life Cycle Mode=::NewEnumerator1
    Loop Delay=&Add_Float(A=$System.Delay, B=1)
    Loop Duration=1
    UseLoopDelay=true
   )
   &SpawnRate(
    Spawn Probability=&Lerp_Float(A=0, B=1, Alpha=$System.InterpolateOverTime.InterpolateOverTime_Float.Moving Average)
    SpawnRate=10
    Use Spawn Probability=true
   )
  ]
  particleSpawnScript [
   &PartitionParticles()
   &InitializeParticle(
    Color=&LerpLinearColors(StartColor=<rgba 0.01 0.01 0.01 1.3>, EndColor=<rgba 0.024 0.008 0.04 1>, LerpFactor=$User.VariantSwitch)
    Lifetime=&Multiply_Float(A=&RandomRangeFloat(Minimum=1, Maximum=2), B=1)
    Sprite Rotation Mode=::NewEnumerator2
    Sprite Size Mode=::NewEnumerator1
    Uniform Sprite Size Max=100
    Uniform Sprite Size Min=80
    UsePositionOffset=false
    Write Material Random=true
   )
  ]
  particleUpdateScript [
   &SkeletalMeshLocation(
    Bone Sampling Mode=::NewEnumerator4
    MaskSpawnGroup=false
    Orientation Sampling=::NewEnumerator1
    Sampled Position Offset=&AddVectorToPosition(
     Position=&AddVectorToPosition(
      Position=&RandomRangeVector(Minimum=<65 0 -10>, Maximum=<65 0 -50>)
      Vector=&Lerp_Vector(
       A=<0 0 0>
       B=&Multiply_VectorByFloat(
        Vector=&SelectVectorFromArray(
         Direct Array Index=$Particles.PartitionParticles.Partition
         Vector Selection Array=DataInterface() {propertyValues {FloatData [<150 230 200>, <-150 230 200>]}}
        )
        Float=&Multiply_Float(
         A=1.2
         B=&MakeCustomFloatFromBool(
          Bool=&SetBoolByIntComparison(A=$Particles.PartitionParticles.Partition, B=0, Comparison Type=::NewEnumerator0)
          True Float=&Lerp_Float(A=1, B=0.8, Alpha=$System.InterpolateOverTime004.InterpolateOverTime_Float.Moving Average)
          False Float=&Lerp_Float(A=1, B=0.8, Alpha=$System.InterpolateOverTime005.InterpolateOverTime_Float.Moving Average)
         )
        )
       )
       Alpha=$Particles.NormalizedAge
      )
     )
     Vector=&SelectVectorFromArray(
      Direct Array Index=$Particles.PartitionParticles.Partition
      Vector Selection Array=DataInterface() {propertyValues {FloatData [<0 0 0>, <-130 0 0>]}}
     )
    )
    Skeletal Mesh=$System.Skeletal Mesh
   ) {module=SkeletalMeshLocation001}
   &ParticleState()
   &SpriteRotationRate(Rotation Rate=&RandomRangeFloat(Minimum=-150, Maximum=150))
   &ScaleSpriteSize(Uniform Curve Sprite Scale=DataInterface() {propertyValues {Curve {keys [
        {arriveTangent=1.6237, interpMode=RCIM_Cubic, leaveTangent=1.6237, tangentMode=RCTM_User, time=0, value=0.8}
        {arriveTangent=0.637, tangentMode=RCTM_User, time=1, value=1.5}
       ]}}})
   &Drag(Drag=1.2, Use Linear Drag=true, Use Rotational Drag=true)
   &GravityForce(Gravity=<0 0 200>)
   &SolveForcesAndVelocity(Clamp Velocity=false, Limit Acceleration=false)
   &ScaleColor(
    Scale Alpha=&FloatFromCurve(
     FloatCurve=DataInterface() {propertyValues {Curve {keys [
         {interpMode=RCIM_Cubic, tangentMode=RCTM_User, time=0, value=0}
         {arriveTangent=-0.5, interpMode=RCIM_Cubic, leaveTangent=-1.8066, tangentMode=RCTM_Break, time=0.6, value=1}
         <at 1 0 RCIM_Cubic>
        ]}}}
     Scale Curve=&Lerp_Float(A=0, B=1, Alpha=$System.InterpolateOverTime.InterpolateOverTime_Float.Moving Average)
    )
    ScaleA=true
    ScaleRGB=true
   )
  ]
  rendererClasses [&NiagaraSpriteRendererProperties]
  renderers [{rendererClass=&NiagaraSpriteRendererProperties}]
 }
 Smoke_Sprite_green {
  emitterUpdateScript [
   &EmitterState(
    Delay First Loop Only=true
    Inactive Response=::NewEnumerator1
    Life Cycle Mode=::NewEnumerator1
    Loop Delay=&Add_Float(A=$System.Delay, B=1)
    Loop Duration=1
    UseLoopDelay=true
   )
   &SpawnRate(
    Spawn Probability=&Lerp_Float(A=0, B=1, Alpha=$System.InterpolateOverTime.InterpolateOverTime_Float.Moving Average)
    SpawnRate=10
    Use Spawn Probability=true
   )
  ]
  particleSpawnScript [
   &PartitionParticles()
   &InitializeParticle(
    Color=&LerpLinearColors(StartColor=<rgba 0.0298 0.13 0.0532 1>, EndColor=<rgba 0.165 0.03 0.3 1>, LerpFactor=$User.VariantSwitch)
    Sprite Rotation Mode=::NewEnumerator2
    Sprite Size Mode=::NewEnumerator1
    Uniform Sprite Size Max=100
    Uniform Sprite Size Min=80
    UsePositionOffset=false
    Write Material Random=true
   )
  ]
  particleUpdateScript [
   &SkeletalMeshLocation(
    Bone Sampling Mode=::NewEnumerator4
    MaskSpawnGroup=false
    Orientation Sampling=::NewEnumerator1
    Sampled Position Offset=&AddVectorToPosition(
     Position=&AddVectorToPosition(
      Position=&RandomRangeVector(Minimum=<65 0 -10>, Maximum=<65 0 -50>)
      Vector=&Lerp_Vector(
       A=<0 0 0>
       B=&Multiply_VectorByFloat(
        Vector=&SelectVectorFromArray(
         Direct Array Index=$Particles.PartitionParticles.Partition
         Vector Selection Array=DataInterface() {propertyValues {FloatData [<150 230 200>, <-150 230 200>]}}
        )
        Float=1.2
       )
       Alpha=$Particles.NormalizedAge
      )
     )
     Vector=&SelectVectorFromArray(
      Direct Array Index=$Particles.PartitionParticles.Partition
      Vector Selection Array=DataInterface() {propertyValues {FloatData [<0 0 0>, <-130 0 0>]}}
     )
    )
    Skeletal Mesh=$System.Skeletal Mesh
   ) {module=SkeletalMeshLocation001}
   &ParticleState()
   &SpriteRotationRate(Rotation Rate=&RandomRangeFloat(Minimum=-150, Maximum=150))
   &ScaleSpriteSize(Uniform Curve Sprite Scale=DataInterface() {propertyValues {Curve {keys [
        {arriveTangent=1.6237, interpMode=RCIM_Cubic, leaveTangent=1.6237, tangentMode=RCTM_User, time=0, value=0.8}
        {arriveTangent=0.637, tangentMode=RCTM_User, time=1, value=1.5}
       ]}}})
   &Drag(Drag=1.2, Use Linear Drag=true, Use Rotational Drag=true)
   &GravityForce(Gravity=<0 0 200>)
   &SolveForcesAndVelocity(Clamp Velocity=false, Limit Acceleration=false)
   &ScaleColor(
    Scale Alpha=&FloatFromCurve(
     FloatCurve=DataInterface() {propertyValues {Curve {keys [
         {interpMode=RCIM_Cubic, tangentMode=RCTM_User, time=0, value=0}
         {arriveTangent=-0.5, interpMode=RCIM_Cubic, leaveTangent=-3.148, tangentMode=RCTM_Break, time=0.3, value=0.85}
         <at 1 0 RCIM_Cubic>
        ]}}}
     Scale Curve=&MakeCustomFloatFromBool(
      Bool=&SetBoolByIntComparison(A=$Particles.PartitionParticles.Partition, B=0, Comparison Type=::NewEnumerator0)
      True Float=&Lerp_Float(A=1, B=0, Alpha=$System.InterpolateOverTime004.InterpolateOverTime_Float.Moving Average)
      False Float=&Lerp_Float(A=1, B=0, Alpha=$System.InterpolateOverTime005.InterpolateOverTime_Float.Moving Average)
     )
    )
    ScaleA=true
    ScaleRGB=true
   )
  ]
  rendererClasses [&NiagaraSpriteRendererProperties]
  renderers [{rendererClass=&NiagaraSpriteRendererProperties}]
 }
 Sparks001 {
  simTarget=GPUComputeSim
  emitterUpdateScript [
   &EmitterState(
    Delay First Loop Only=true
    Inactive Response=::NewEnumerator1
    Life Cycle Mode=::NewEnumerator1
    Loop Delay=&Add_Float(A=$System.Delay, B=0.5)
    Loop Duration=1
    UseLoopDelay=true
   )
   &SpawnRate(
    Spawn Probability=&Lerp_Float(A=0, B=1, Alpha=$System.InterpolateOverTime.InterpolateOverTime_Float.Moving Average)
    SpawnRate=100
    Use Spawn Probability=true
   )
  ]
  particleSpawnScript [
   &PartitionParticles()
   &InitializeParticle(
    Color=&LerpLinearColors(StartColor=<rgba 0.5647 5 1.5996 1>, EndColor=<rgba 5.5647 1.1293 10 1>, LerpFactor=$User.VariantSwitch)
    Lifetime Max=0.5
    Lifetime Min=0.5
    Lifetime Mode=::NewEnumerator1
    Sprite Size Max=<1 7>
    Sprite Size Min=<1 5>
    Sprite Size Mode=::NewEnumerator2
    UsePositionOffset=false
    Write Material Random=true
   )
   &SkeletalMeshLocation(
    Bone Sampling Mode=::NewEnumerator4
    MaskSpawnGroup=false
    Orientation Sampling=::NewEnumerator1
    Sampled Position Offset=&AddVectorToPosition(
     Position=&SelectVectorFromArray(
      Direct Array Index=$Particles.PartitionParticles.Partition
      Vector Selection Array=DataInterface() {propertyValues {FloatData [<-100 -20 15>, <80 -20 15>]}}
     )
     Vector=<0 0 0>
    )
    Skeletal Mesh=$System.Skeletal Mesh
   )
   &ShapeLocation(MaskSpawnGroup=false, Override Seed=false, Sphere Radius=20, UseOwnerScale=true)
   &AddVelocity(
    Cone Angle=50
    Cone Axis=&SelectVectorFromArray(
     Direct Array Index=$Particles.PartitionParticles.Partition
     Vector Selection Array=DataInterface() {propertyValues {FloatData [<1 -0.8 0.7>, <-1 -0.8 0.7>]}}
    )
    Module.UseConeFalloff=false
    Module.UseVelDistribution=false
    Override Seed=false
    Velocity Mode=::NewEnumerator2
    Velocity Speed=&RandomRangeFloat(Minimum=150, Maximum=600)
   )
   &CameraOffset(Camera Offset Amount=-50)
  ]
  particleUpdateScript [
   &ParticleState()
   /SetVariables(Particles.Glow03_Color=&MakeLinearColorFromVectorAndFloat(
     "Vector (RGB)"=&MakeVectorFromLinearColorRGB(Color=&MultiplyLinearColor(
       Linear Color=&LerpLinearColors(StartColor=<rgba 0.0335 1 0.2591 1>, EndColor=<rgba 1.5503 0.1006 3 1>, LerpFactor=$User.VariantSwitch)
       Scale Factor=&MakeVector4_FromFloat(Float=1)
      ))
     "Float (Alpha)"=&Multiply_Float(A=0.2, B=1)
    )) {isSetParameters=true, module=SetVariables_#6}
   /SetVariables(Particles.Glow03_Size=&Multiply_Vector2D_ByFloat(
     Float=&FloatFromCurve(
      FloatCurve=DataInterface() {propertyValues {Curve {keys [
          <at 0 0 RCIM_Cubic>
          <at 0.2 1.3 RCIM_Cubic>
          {arriveTangent=-0.1423, interpMode=RCIM_Cubic, leaveTangent=-0.1423, tangentMode=RCTM_User, time=1, value=0}
         ]}}}
      Scale Curve=10
     )
     Vector2D=<1 4>
    )) {isSetParameters=true, module=SetVariables_#7}
   &ScaleColor(
    Scale Alpha=&FloatFromCurve(
     FloatCurve=DataInterface() {propertyValues {Curve {keys [<at 0 1>,<at 0.3 1>,<at 1 1>]}}}
     Scale Curve=&MakeCustomFloatFromBool(
      Bool=&SetBoolByIntComparison(A=$Particles.PartitionParticles.Partition, B=0, Comparison Type=::NewEnumerator0)
      True Float=&Lerp_Float(A=1, B=0, Alpha=$System.InterpolateOverTime005.InterpolateOverTime_Float.Moving Average)
      False Float=&Lerp_Float(A=1, B=0, Alpha=$System.InterpolateOverTime004.InterpolateOverTime_Float.Moving Average)
     )
    )
    ScaleA=true
    ScaleRGB=true
   )
   &ScaleSpriteSize(Scale Factor=&Vector2DFromFloat(Value=&FloatFromCurve(
      FloatCurve=DataInterface() {propertyValues {Curve {keys [
          <at 0 0 RCIM_Cubic>
          {interpMode=RCIM_Cubic, leaveTangent=-3.0296, tangentMode=RCTM_Break, time=0.2, value=1.2}
          {arriveTangent=-1.8107, interpMode=RCIM_Cubic, leaveTangent=-1.8107, tangentMode=RCTM_User, time=1, value=0}
         ]}}}
      Scale Curve=2
     ))) {module=ScaleSpriteSize001}
   &CurlNoiseForce(Noise Frequency=500, Noise Strength=500, Pan Noise=false)
   &Drag(Use Linear Drag=true, Use Rotational Drag=true)
   &SolveForcesAndVelocity(Clamp Velocity=false, Limit Acceleration=false)
   &ScaleSpriteSizeBySpeed(Max Scale Factor=<1 2>)
  ]
  rendererClasses [&NiagaraSpriteRendererProperties]
  renderers [
   {rendererClass=&NiagaraSpriteRendererProperties}
   {rendererClass=&NiagaraSpriteRendererProperties, rendererIndex=1}
  ]
 }
 Sub_Smoke {
  emitterUpdateScript [
   &EmitterState(
    Delay First Loop Only=true
    Inactive Response=::NewEnumerator1
    Life Cycle Mode=::NewEnumerator1
    Loop Delay=$System.Delay
    Loop Duration=1
    UseLoopDelay=true
   )
   &SpawnRate(
    SpawnRate=&Multiply_Float(A=8, B=&Lerp_Float(A=0, B=1, Alpha=$System.InterpolateOverTime.InterpolateOverTime_Float.Moving Average))
    Use Spawn Probability=false
   ) {module=SpawnRate001}
   &SpawnRate(
    SpawnRate=&Multiply_Float(A=8, B=&Lerp_Float(A=0, B=1, Alpha=$System.InterpolateOverTime.InterpolateOverTime_Float.Moving Average))
    Use Spawn Probability=false
   ) {enabled=false}
  ]
  particleSpawnScript [
   &PartitionParticles()
   &InitializeParticle(
    Color=&LerpLinearColors(StartColor=<rgba 0.0266 0.2 0.0671 1>, EndColor=<rgba 0.17 0.0399 0.3 1>, LerpFactor=$User.VariantSwitch)
    Lifetime Max=1.5
    Lifetime Mode=::NewEnumerator1
    Sprite Rotation Mode=::NewEnumerator2
    Sprite Size Mode=::NewEnumerator1
    Sprite UV Mode=::NewEnumerator3
    Uniform Sprite Size Max=80
    Uniform Sprite Size Min=60
    UsePositionOffset=false
    Write Material Random=true
   )
   &AddVelocity(
    Cone Angle=10
    Cone Axis=<0.2 1 0.5>
    Module.UseConeFalloff=false
    Module.UseVelDistribution=false
    Override Seed=false
    Velocity Mode=::NewEnumerator2
    Velocity Speed=&RandomRangeFloat(Minimum=200, Maximum=300)
   )
  ]
  particleUpdateScript [
   &SkeletalMeshLocation(
    Bone Sampling Mode=::NewEnumerator4
    MaskSpawnGroup=false
    Sampled Position Offset=&AddVectorToPosition(
     Position=&AddVectorToPosition(
      Position=<65 -20 -15>
      Vector=&Multiply_VectorByFloat(
       Vector=&Lerp_Vector(
        A=<0 0 0>
        B=&SelectVectorFromArray(
         Direct Array Index=$Particles.PartitionParticles.Partition
         Vector Selection Array=DataInterface() {propertyValues {FloatData [<50 200 100>, <-50 200 100>]}}
        )
        Alpha=$Particles.NormalizedAge
       )
       Float=1.2
      )
     )
     Vector=&SelectVectorFromArray(
      Direct Array Index=$Particles.PartitionParticles.Partition
      Vector Selection Array=DataInterface() {propertyValues {FloatData [<0 0 0>, <-150 0 0>]}}
     )
    )
    Skeletal Mesh=$System.Skeletal Mesh
   ) {module=SkeletalMeshLocation001}
   &ParticleState()
   &SpriteRotationRate(Rotation Rate=&RandomRangeFloat(Minimum=-100, Maximum=100))
   &ScaleSpriteSize(Uniform Curve Sprite Scale=DataInterface() {propertyValues {Curve {keys [
        {arriveTangent=1.6237, interpMode=RCIM_Cubic, leaveTangent=1.6237, tangentMode=RCTM_User, time=0, value=1}
        {arriveTangent=0.637, tangentMode=RCTM_User, time=1, value=1}
       ]}}})
   &Drag(Drag=2, Use Linear Drag=true, Use Rotational Drag=true)
   &SolveForcesAndVelocity(Clamp Velocity=false, Limit Acceleration=false)
   &ScaleColor(
    Scale Alpha=&FloatFromCurve(
     FloatCurve=DataInterface() {propertyValues {Curve {keys [
         {interpMode=RCIM_Cubic, tangentMode=RCTM_User, time=0, value=0}
         {arriveTangent=-0.5, interpMode=RCIM_Cubic, leaveTangent=-3.5948, tangentMode=RCTM_Break, time=0.4, value=1.05}
         <at 1 0 RCIM_Cubic>
        ]}}}
     Scale Curve=&Multiply_Float(
      A=&Lerp_Float(A=0, B=1, Alpha=$System.InterpolateOverTime.InterpolateOverTime_Float.Moving Average)
      B=&MakeCustomFloatFromBool(
       Bool=&SetBoolByIntComparison(A=$Particles.PartitionParticles.Partition, B=0, Comparison Type=::NewEnumerator0)
       True Float=&Lerp_Float(A=1, B=0, Alpha=$System.InterpolateOverTime004.InterpolateOverTime_Float.Moving Average)
       False Float=&Lerp_Float(A=1, B=0, Alpha=$System.InterpolateOverTime005.InterpolateOverTime_Float.Moving Average)
      )
     )
    )
    Scale RGB=&MakeVectorFromLinearColorRGB(Color=&ColorFromCurve(ColorCurve=DataInterface() {propertyValues {
        AlphaCurve {keys []}
        BlueCurve {keys [<at 0 1 RCIM_Cubic>,<at 0.3549 0.2 RCIM_Cubic>]}
        GreenCurve {keys [<at 0 1 RCIM_Cubic>,<at 0.3549 0.2 RCIM_Cubic>]}
        RedCurve {keys [<at 0 1 RCIM_Cubic>,<at 0.3549 0.2 RCIM_Cubic>]}
       }}))
    ScaleA=true
    ScaleRGB=true
   )
   &SubUVAnimation(
    End Frame Range Override=30
    "Sprite Renderer "=DataInterface() {propertyValues {SpriteRenderer=&self:Sub_Smoke_6.NiagaraSpriteRendererProperties_0}}
    SubUV Lookup Index Scale=1.5
    UseEndFrame=true
    UseStartFrame=true
   )
  ]
  rendererClasses [&NiagaraSpriteRendererProperties]
  renderers [{rendererClass=&NiagaraSpriteRendererProperties}]
 }
 Wheel_Small_Smoke {
  emitterUpdateScript [
   &EmitterState(
    Delay First Loop Only=true
    Inactive Response=::NewEnumerator1
    Life Cycle Mode=::NewEnumerator1
    Loop Delay=$System.Delay
    Loop Duration=1
    UseLoopDelay=true
   )
   &SpawnRate(
    Spawn Probability=&MakeFloatFromInt(Int=&MakeCustomIntFromBool(Bool=$User.bIsFrontEndPreview, True INT=0, False INT=1))
    SpawnRate=40
    Use Spawn Probability=true
   )
  ]
  particleSpawnScript [
   &PartitionParticles()
   &InitializeParticle(
    Color=&LerpLinearColors(StartColor=<rgba 0.01 0.01 0.01 1.7>, EndColor=<rgba 0.05 0 0.1 1.7>, LerpFactor=$User.VariantSwitch)
    Lifetime Max=1.5
    Lifetime Min=0.7
    Lifetime Mode=::NewEnumerator1
    Sprite Rotation Mode=::NewEnumerator2
    Sprite Size Mode=::NewEnumerator1
    Uniform Sprite Size Max=100
    Uniform Sprite Size Min=80
    UsePositionOffset=false
    Write Material Random=true
   )
   &AddVelocity(
    Cone Angle=10
    Cone Axis=<-1 -1 1>
    Module.UseConeFalloff=false
    Module.UseVelDistribution=false
    Override Seed=false
    Velocity Mode=::NewEnumerator2
    Velocity Speed=&RandomRangeFloat(Minimum=200, Maximum=300)
   )
  ]
  particleUpdateScript [
   &SkeletalMeshLocation(
    Bone Sampling Mode=::NewEnumerator4
    MaskSpawnGroup=false
    Sampled Position Offset=&AddVectorToPosition(
     Position=&AddVectorToPosition(
      Position=&AddVectorToPosition(
       Position=<65 -20 -10>
       Vector=&MakeCustomVectorFromBool(Bool=&RandomBool(), True Vector=<0 0 30>, False Vector=<1 0 -20>)
      )
      Vector=&Lerp_Vector(
       A=<0 0 0>
       B=&SelectVectorFromArray(
        Direct Array Index=$Particles.PartitionParticles.Partition
        Vector Selection Array=DataInterface() {propertyValues {FloatData [<50 200 70>, <-50 200 70>]}}
       )
       Alpha=$Particles.NormalizedAge
      )
     )
     Vector=&SelectVectorFromArray(
      Direct Array Index=$Particles.PartitionParticles.Partition
      Vector Selection Array=DataInterface() {propertyValues {FloatData [<0 0 0>, <-150 0 0>]}}
     )
    )
    Skeletal Mesh=$System.Skeletal Mesh
   )
   &ParticleState()
   &SpriteRotationRate(Rotation Rate=&RandomRangeFloat(Minimum=-150, Maximum=150))
   &ScaleSpriteSize(Uniform Curve Sprite Scale=DataInterface() {propertyValues {Curve {keys [
        {arriveTangent=1.6237, interpMode=RCIM_Cubic, leaveTangent=1.6237, tangentMode=RCTM_User, time=0, value=1}
        {arriveTangent=0.637, tangentMode=RCTM_User, time=1, value=1.7}
       ]}}})
   &Drag(Drag=2, Use Linear Drag=true, Use Rotational Drag=true)
   &SolveForcesAndVelocity(Clamp Velocity=false, Limit Acceleration=false)
   &ScaleColor(
    Scale Alpha=&FloatFromCurve(FloatCurve=DataInterface() {propertyValues {Curve {keys [
         {interpMode=RCIM_Cubic, tangentMode=RCTM_User, time=0, value=0}
         {arriveTangent=-0.5, interpMode=RCIM_Cubic, leaveTangent=-3.148, tangentMode=RCTM_Break, time=0.2, value=0.9}
         <at 1 0 RCIM_Cubic>
        ]}}})
    ScaleA=true
    ScaleRGB=true
   )
  ]
  rendererClasses [&NiagaraSpriteRendererProperties]
  renderers [{rendererClass=&NiagaraSpriteRendererProperties}]
 }
 Wheel_glow {
  emitterUpdateScript [
   &EmitterState(
    Inactive Response=::NewEnumerator1
    Life Cycle Mode=::NewEnumerator1
    Loop Behavior=::NewEnumerator1
    Loop Delay=$System.Delay
    Loop Duration Mode=::NewEnumerator1
    UseLoopDelay=true
   )
   &SpawnBurst_Instantaneous(Spawn Count=2, Use Spawn Probability=false, UseLoopCountLimit=false)
  ]
  particleSpawnScript [
   &PartitionParticles()
   &InitializeParticle(
    Color=&LerpLinearColors(StartColor=<rgba 0.5692 2 0.9972 2>, EndColor=<rgba 2.2 0.4 4 2>, LerpFactor=$User.VariantSwitch)
    Mesh Scale=&Multiply_VectorByFloat(
     Vector=&SelectVectorFromArray(
      Direct Array Index=$Particles.PartitionParticles.Partition
      Vector Selection Array=DataInterface() {propertyValues {FloatData [<1 1 1>, <1 1 1>]}}
     )
     Float=1
    )
    Mesh Scale Mode=::NewEnumerator4
    Sprite Size Mode=::NewEnumerator3
    UsePositionOffset=false
    Write Material Random=true
   )
   &DynamicMaterialParameters(
    Index 0 Param 1=30
    Index 0 Param 2=30
    Module.Index0Param0Write=true
    Module.Index0Param1Write=true
    Module.Index0Param2Write=true
    Module.Index0Param3Write=true
   )
   &InitialMeshOrientation(Mesh Orientation Mode=::NewEnumerator2, Rotation=<-0.3 0 0>, Use Rotation Vector=true)
   &CameraOffset(Camera Offset Amount=10)
  ]
  particleUpdateScript [
   &ParticleState(Kill Particles When Lifetime Has Elapsed=false, Let Infinitely Lived Particles Die When Emitter Deactivates=true)
   &SkeletalMeshLocation(
    Bone Sampling Mode=::NewEnumerator4
    MaskSpawnGroup=false
    Sampled Position Offset=&SelectVectorFromArray(
     Direct Array Index=$Particles.PartitionParticles.Partition
     Vector Selection Array=DataInterface() {propertyValues {FloatData [<-75 -23 -11>, <75 -23 -11>]}}
    )
    Skeletal Mesh=$System.Skeletal Mesh
   )
   &ScaleColor(
    Scale Alpha=&Lerp_Float(A=0, B=1, Alpha=$System.InterpolateOverTime.InterpolateOverTime_Float.Moving Average)
    ScaleA=true
    ScaleRGB=true
   )
   &FN_Light_Attributes(
    Color=<rgba 0 30 5 1>
    Create Color Attribute=true
    Create Exponent Attribute=true
    Create Position Attribute=true
    Create Radius Attribute=true
    Exponent=3
    Position=&AddVectorToPosition(
     Position=$Particles.Position
     Vector=&TransformVector(
      Vector=&SelectVectorFromArray(
       Direct Array Index=$Particles.PartitionParticles.Partition
       Vector Selection Array=DataInterface() {propertyValues {FloatData [<-20 -20 0>, <20 -20 0>]}}
      )
      Destination Space=::Simulation
     )
    )
    Radius=100
    Scale Color=&Lerp_Float(A=1, B=1, Alpha=$System.InterpolateOverTime.InterpolateOverTime_Float.Moving Average)
   )
  ]
  rendererClasses [&NiagaraMeshRendererProperties]
  renderers [{rendererClass=&NiagaraMeshRendererProperties, rendererIndex=0}]
 }
 Wheel_skull {
  emitterUpdateScript [
   &EmitterState(
    Inactive Response=::NewEnumerator1
    Life Cycle Mode=::NewEnumerator1
    Loop Behavior=::NewEnumerator1
    Loop Duration Mode=::NewEnumerator1
    UseLoopDelay=true
   )
   &SpawnBurst_Instantaneous(
    Spawn Count=18
    Spawn Probability=&MakeFloatFromInt(Int=&MakeCustomIntFromBool(Bool=$User.bIsFrontEndPreview, True INT=0, False INT=1))
    Use Spawn Probability=true
    UseLoopCountLimit=false
   )
  ]
  particleSpawnScript [
   &PartitionParticles()
   &InitializeParticle(
    Color=&LerpLinearColors(StartColor=<rgba 0.2 0.2 0.2 2>, EndColor=<rgba 0.13 0.06 0.2 2>, LerpFactor=$User.VariantSwitch)
    Mesh Scale=&Multiply_VectorByFloat(Vector=<1 1 1>, Float=3)
    Mesh Scale Mode=::NewEnumerator4
    UsePositionOffset=false
    Write Material Random=true
   )
   &InitialMeshOrientation(Mesh Orientation Mode=::NewEnumerator2, Use Rotation Vector=true)
   &CameraOffset()
  ]
  particleUpdateScript [
   &ParticleState(Kill Particles When Lifetime Has Elapsed=false, Let Infinitely Lived Particles Die When Emitter Deactivates=true)
   &SkeletalMeshLocation(
    Bone Sampling Mode=::NewEnumerator4
    MaskSpawnGroup=false
    Orientation Sampling=::NewEnumerator1
    Position Sampling=::NewEnumerator2
    Sampled Orientation Coordinate Space=::Local
    Skeletal Mesh=$System.Skeletal Mesh
   ) {module=SkeletalMeshLocation001}
   /SetVariables(Particles.CenterOffset:Vector3f=&Multiply_Vector(
     A=&VectorFromFloat(Value=&MakeFloatFromVector(Vector=$Particles.SkeletalMeshLocation001.SampledScale, Channel=::NewEnumerator0))
     B=&SelectVectorFromArray(
      Direct Array Index=$Particles.PartitionParticles.Partition
      Vector Selection Array=DataInterface() {propertyValues {FloatData [<75 -22 -8>, <-75 -22 -8>]}}
     )
    )) {isSetParameters=true}
   &ShapeLocation(
    Handle Radius=0
    Large Radius=&Multiply_Float(
     A=30
     B=&MakeFloatFromVector(Vector=$Particles.SkeletalMeshLocation001.SampledScale, Channel=::NewEnumerator0)
    )
    MaskSpawnGroup=false
    Rotation Mode=::NewEnumerator1
    Shape Origin=&VectorFromFloat(Value=0)
    Shape Primitive=::NewEnumerator3
    Torus Distribution Mode:ENiagaraTorusDistributionMode=::NewEnumerator1
    U Position=&ReturnNormalizedExecIndex()
    UseOwnerScale=true
   )
   &ScaleMeshSize(Scale Factor=&Multiply_VectorByFloat(
     Vector=<1 1 1>
     Float=&Multiply_Float(
      A=0.4
      B=&MakeFloatFromVector(Vector=$Particles.SkeletalMeshLocation001.SampledScale, Channel=::NewEnumerator0)
     )
    ))
   &SkeletalMeshLocation(
    Bone Sampling Mode=::NewEnumerator4
    MaskSpawnGroup=false
    Orientation Sampling=::NewEnumerator1
    Sampled Orientation Coordinate Space=::Local
    Sampled Position Offset=&AddVectorToPosition(Position=$Particles.ShapeLocation.ShapePosition, Vector=$Particles.CenterOffset)
    Skeletal Mesh=$System.Skeletal Mesh
   )
   &SolveForcesAndVelocity(Clamp Velocity=false, Limit Acceleration=false)
   &self:ScratchModule(
    RotAxis:Vector3f=&RotateVectorByQuaternion(Vector To Rotate=<1 0 0>, Quaternion=$Particles.SkeletalMeshLocation.SampledRotation)
    RotSpeed:NiagaraFloat=&Lerp_Float(A=3, B=6, Alpha=$System.InterpolateOverTime.InterpolateOverTime_Float.Moving Average)
    SocketPos:NiagaraPosition=&AddVectorToPosition(
     Position=$Output.SkeletalMeshLocation.SampledPositionWithoutOffset
     Vector=&RotateVectorByQuaternion(Vector To Rotate=$Particles.CenterOffset, Quaternion=$Particles.SkeletalMeshLocation.SampledRotation)
    )
   ) {module=ScratchModule}
   /SetVariables(Particles.MeshOrientation=&MultiplyQuaternion(
     Quaternion A=&FindQuatBetweenTwoVectors(
      From Vector=<0 1 0>
      To Vector=&NormalizeVector(
       Vector To Normalize=&SubtractVectorFromPosition(
        Position=$Particles.OPosition
        Vector=&AddVectorToPosition(
         Position=$Output.SkeletalMeshLocation.SampledPositionWithoutOffset
         Vector=&RotateVectorByQuaternion(Vector To Rotate=$Particles.CenterOffset, Quaternion=$StackContext.SkeletalMeshLocation.SampledRotation)
        )
       )
       Safe Normalize=true
       Safe Fallback Threshold=0
       Safe Fallback Vector=<0 0 0.001>
      )
     )
     Quaternion B=&MakeQuatFromAxisAngle(Axis=<1 0 0>, AngleInDegrees=-60)
    )) {enabled=false, isSetParameters=true}
   &UpdateMeshOrientation(
    Facing Interpolation=false
    "Facing Origin ":NiagaraPosition=&AddVectorToPosition(
     Position=$Output.SkeletalMeshLocation.SampledPositionWithoutOffset
     Vector=&RotateVectorByQuaternion(Vector To Rotate=$Particles.CenterOffset, Quaternion=$Particles.SkeletalMeshLocation001.SampledRotation)
    )
    Facing Origin Coordinate Space:ENiagaraCoordinateSpace=::Local
    Facing Position:NiagaraPosition=$Particles.OPosition
    Facing Position Coordinate Space:ENiagaraCoordinateSpace=::Local
    Orientation Method=::NewEnumerator2
    Override Quat=false
    TransformFacingOrigin=true
    TransformFacingPosition=true
    TransformUp=true
    Up Coordinate Space:ENiagaraCoordinateSpace=::Local
    Up Direction:Vector3f=&RotateVectorByQuaternion(Vector To Rotate=<-1 0 0>, Quaternion=$Particles.SkeletalMeshLocation001.SampledRotation)
   )
   &ScaleColor(
    Scale Alpha=&Multiply_Float(A=1, B=1)
    Scale RGB=&MakeVectorFromLinearColorRGB(Color=&LerpLinearColors(
      StartColor=<rgba 1 1 1 1>
      EndColor=&LerpLinearColors(StartColor=<rgba 0.1146 2 0.696 1>, EndColor=<rgba 2 2 2 1>, LerpFactor=$User.VariantSwitch)
      LerpFactor=$System.InterpolateOverTime.InterpolateOverTime_Float.Moving Average
     ))
    ScaleA=true
    ScaleRGB=true
   )
   &DynamicMaterialParameters(
    Index 0 Param 4=&Lerp_Float(A=0, B=1, Alpha=$System.InterpolateOverTime.InterpolateOverTime_Float.Moving Average)
    Module.Index0Param0Write=false
    Module.Index0Param1Write=false
    Module.Index0Param2Write=false
    Module.Index0Param3Write=true
   )
  ]
  rendererClasses [&NiagaraMeshRendererProperties]
  renderers [
   {rendererClass=&NiagaraMeshRendererProperties, rendererIndex=0}
   {rendererClass=&NiagaraMeshRendererProperties, rendererIndex=1}
  ]
 }
}
summary {
 systemName=NS_Glider_DarkStance
 emitters [
  {emitterName=Smoke_Sprite_dark, rendererClasses [&NiagaraSpriteRendererProperties]}
  {emitterName=Wheel_Small_Smoke, rendererClasses [&NiagaraSpriteRendererProperties]}
  {emitterName=Sub_Smoke, rendererClasses [&NiagaraSpriteRendererProperties]}
  {emitterName=Smoke_Dark, rendererClasses [&NiagaraMeshRendererProperties]}
  {emitterName=Green_Smoke, rendererClasses [&NiagaraMeshRendererProperties]}
  {emitterName=Wheel_glow, rendererClasses [&NiagaraMeshRendererProperties]}
  {emitterName=Skull_Sprite, rendererClasses [&NiagaraMeshRendererProperties, &NiagaraSpriteRendererProperties]}
  {emitterName=Burst_Sparks, simTarget=GPUComputeSim, rendererClasses [&NiagaraSpriteRendererProperties]}
  {emitterName=Burst_Sub_Smoke, rendererClasses [&NiagaraSpriteRendererProperties]}
  {emitterName=Flash, rendererClasses [&NiagaraSpriteRendererProperties]}
  {emitterName=Smoke_Sprite_green, rendererClasses [&NiagaraSpriteRendererProperties]}
  {emitterName=Smoke_Dark_Flash, rendererClasses [&NiagaraMeshRendererProperties]}
  {emitterName=Sparks001, simTarget=GPUComputeSim, rendererClasses [&NiagaraSpriteRendererProperties]}
  {emitterName=Light, rendererClasses [&NiagaraLightRendererProperties]}
  {emitterName=Wheel_skull, rendererClasses [&NiagaraMeshRendererProperties]}
  {emitterName=Burst_Sparks001, simTarget=GPUComputeSim, rendererClasses [&NiagaraSpriteRendererProperties]}
  {emitterName=Burst_Sub_Smoke001, rendererClasses [&NiagaraSpriteRendererProperties]}
  {emitterName=Flash001, rendererClasses [&NiagaraSpriteRendererProperties]}
 ]
 userVariables [
  User.bIsFrontEndPreview {type=NiagaraBool}
  User.ForwardDot {type=NiagaraFloat, defaultValue=0, description=""}
  User.RightDot {type=NiagaraFloat, defaultValue=0, description=""}
  User.VariantSwitch {type=NiagaraFloat, defaultValue=0, description=""}
 ]
}
