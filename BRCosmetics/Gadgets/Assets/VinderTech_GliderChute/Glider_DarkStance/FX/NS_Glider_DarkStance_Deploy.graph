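# Niagara deploy effect for the Dark Stance glider, trailing sprite smoke from chariot and horse sockets on its skeletal mesh.
# User.bIsFrontEndPreview gates spawn probability between in-game emitters and FE_ preview copies, which use loop delay and owner-scaled sprites.
# Horse smoke partitions particles across four sockets, with in-game spawn rate driven by a system-level smoothed value.
# User.VariantSwitch lerps colors from green/black to purple; the Sparkles emitter is disabled.
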NiagaraSystem /BRCosmetics/Gadgets/Assets/VinderTech_GliderChute/Glider_DarkStance/FX/NS_Glider_DarkStance_Deploy
dependencies {
 usedDataInterfaces [
  &NiagaraDataInterfaceCurve
  &NiagaraDataInterfaceSpriteRendererInfo
  &NiagaraDataInterfaceSkeletalMesh
 ]
 usedDynamicInputs [&MakeCustomFloatFromBool, &ClampFloat]
 usedModules [
  &self:SystemSpawnScript.SystemScriptSource.SystemScriptGraph.NiagaraNodeAssignment_0.SetVariables_#1
  &self:SystemSpawnScript.SystemScriptSource.SystemScriptGraph.NiagaraNodeAssignment_1.SetVariables_#2
  &self:SystemSpawnScript.SystemScriptSource.SystemScriptGraph.NiagaraNodeAssignment_2.SetVariables_#3
  &SystemState
  &InterpolateOverTime
 ]
}
emitters {
 Chariot_Green_Smoke {
  simTarget=GPUComputeSim
  emitterUpdateScript [
   &EmitterState(Life Cycle Mode=::NewEnumerator1, Loop Behavior=::NewEnumerator1, Loop Duration=$System.Duration, UseLoopDelay=false)
   &SpawnRate(
    Spawn Probability=&MakeFloatFromInt(Int=&MakeCustomIntFromBool(Bool=$User.bIsFrontEndPreview, True INT=0, False INT=1))
    SpawnRate=7
    Use Spawn Probability=true
   )
  ]
  particleSpawnScript [
   &PartitionParticles(Number Of Partitions=1)
   &InitializeParticle(
    Color=&LerpLinearColors(StartColor=<rgba 0.0229 0.1 0.0409 1>, EndColor=<rgba 0.1725 0.045 0.3 1>, LerpFactor=$User.VariantSwitch)
    Lifetime Mode=::NewEnumerator1
    Sprite Rotation Mode=::NewEnumerator2
    Sprite Size Mode=::NewEnumerator1
    Uniform Sprite Size Max=200
    Uniform Sprite Size Min=100
    UsePositionOffset=false
    Write Material Random=true
   )
   &InheritVelocity(Inherited Velocity Amount Scale=<1 1 0>, Inherited Velocity Speed Limit=600, Speed Limit=true, UseSpeedThreshold=false)
  ]
  particleUpdateScript [
   &CameraOffset(Camera Offset Amount=-50)
   &SkeletalMeshLocation(
    Bone / Socket Index=4
    Bone Sampling Mode=::NewEnumerator4
    MaskSpawnGroup=false
    Orientation Sampling=::NewEnumerator1
    Position Sampling=::NewEnumerator1
    Sampled Position Offset=<0 0 -50>
    Sampled Position Offset Coordinate Space:ENiagaraSkelMeshTransforms=::NewEnumerator2
    Skeletal Mesh=$System.Skeletal Mesh
    Source Movement Scale:Vector3f=<0.2 0.2 1>
   )
   &ParticleState()
   &ScaleSpriteSize(Uniform Curve Sprite Scale=DataInterface() {propertyValues {Curve {keys [
        {arriveTangent=1.6237, interpMode=RCIM_Cubic, leaveTangent=1.6237, tangentMode=RCTM_User, time=0, value=1}
        {arriveTangent=0.637, tangentMode=RCTM_User, time=1, value=1.7}
       ]}}})
   &SpriteRotationRate(Rotation Rate=&RandomRangeFloat(Minimum=-150, Maximum=150))
   &SolveForcesAndVelocity(Clamp Velocity=false, Limit Acceleration=false)
   &ScaleColor(
    Scale Alpha=&FloatFromCurve(FloatCurve=DataInterface() {propertyValues {Curve {keys [
         {interpMode=RCIM_Cubic, tangentMode=RCTM_User, time=0, value=0}
         {arriveTangent=-0.5, interpMode=RCIM_Cubic, leaveTangent=-3.148, tangentMode=RCTM_Break, time=0.1, value=0.85}
         <at 1 0 RCIM_Cubic>
        ]}}})
    ScaleA=true
    ScaleRGB=true
   )
  ]
  rendererClasses [&NiagaraSpriteRendererProperties]
  renderers [{rendererClass=&NiagaraSpriteRendererProperties}]
 }
 Chariot_Green_Smoke2 {
  emitterUpdateScript [
   &EmitterState(Life Cycle Mode=::NewEnumerator1, Loop Behavior=::NewEnumerator1, Loop Duration=$System.Duration, UseLoopDelay=false)
   &SpawnRate(
    Spawn Probability=&MakeFloatFromInt(Int=&MakeCustomIntFromBool(Bool=$User.bIsFrontEndPreview, True INT=0, False INT=1))
    SpawnRate=7
    Use Spawn Probability=true
   )
  ]
  particleSpawnScript [
   &PartitionParticles(Number Of Partitions=1)
   &InitializeParticle(
    Color=&LerpLinearColors(StartColor=<rgba 0.0053 0.04 0.0169 0.7>, EndColor=<rgba 0.0397 0.0093 0.07 0.7>, LerpFactor=$User.VariantSwitch)
    Lifetime Min=2
    Lifetime Mode=::NewEnumerator1
    Sprite Rotation Mode=::NewEnumerator2
    Sprite Size Mode=::NewEnumerator1
    Sprite UV Mode=::NewEnumerator3
    Uniform Sprite Size Max=180
    Uniform Sprite Size Min=140
    UsePositionOffset=false
    Write Material Random=true
   )
   &InheritVelocity(Inherited Velocity Amount Scale=<1 1 0>, Inherited Velocity Speed Limit=600, Speed Limit=true, UseSpeedThreshold=false)
  ]
  particleUpdateScript [
   &CameraOffset(Camera Offset Amount=-50)
   &SkeletalMeshLocation(
    Bone / Socket Index=4
    Bone Sampling Mode=::NewEnumerator4
    MaskSpawnGroup=false
    Orientation Sampling=::NewEnumerator1
    Position Sampling=::NewEnumerator1
    Sampled Position Offset=<0 0 -50>
    Sampled Position Offset Coordinate Space:ENiagaraSkelMeshTransforms=::NewEnumerator2
    Skeletal Mesh=$System.Skeletal Mesh
    Source Movement Scale:Vector3f=<0.2 0.2 1>
   )
   &ParticleState()
   &ScaleSpriteSize(Uniform Curve Sprite Scale=DataInterface() {propertyValues {Curve {keys [
        {arriveTangent=0.2792, interpMode=RCIM_Cubic, leaveTangent=0.2792, tangentMode=RCTM_User, time=0, value=1.6}
        {arriveTangent=0.637, tangentMode=RCTM_User, time=1, value=1.7}
       ]}}})
   &SpriteRotationRate(Rotation Rate=&RandomRangeFloat(Minimum=-150, Maximum=150))
   &SolveForcesAndVelocity(Clamp Velocity=false, Limit Acceleration=false)
   &ScaleColor(
    Scale Alpha=&FloatFromCurve(FloatCurve=DataInterface() {propertyValues {Curve {keys [
         {interpMode=RCIM_Cubic, tangentMode=RCTM_User, time=0, value=0}
         {arriveTangent=-0.5, interpMode=RCIM_Cubic, leaveTangent=-3.148, tangentMode=RCTM_Break, time=0.1, value=0.85}
         <at 0.7 0 RCIM_Cubic>
        ]}}})
    ScaleA=true
    ScaleRGB=true
   )
   &SubUVAnimation(
    End Frame Range Override=64
    "Sprite Renderer "=DataInterface() {propertyValues {SpriteRenderer=&self:Chariot_Green_Smoke2_6.NiagaraSpriteRendererProperties_1}}
    SubUV Lookup Index Scale=1.5
    UseEndFrame=true
    UseStartFrame=true
   )
  ]
  rendererClasses [&NiagaraSpriteRendererProperties]
  renderers [{rendererClass=&NiagaraSpriteRendererProperties}]
 }
 Chariot_Smoke001 {
  emitterUpdateScript [
   &EmitterState(Life Cycle Mode=::NewEnumerator1, Loop Behavior=::NewEnumerator1, Loop Duration=$System.Duration, UseLoopDelay=false)
   &SpawnRate(
    Spawn Probability=&MakeFloatFromInt(Int=&MakeCustomIntFromBool(Bool=$User.bIsFrontEndPreview, True INT=0, False INT=1))
    SpawnRate=30
    Use Spawn Probability=true
   )
  ]
  particleSpawnScript [
   &PartitionParticles(Number Of Partitions=1)
   &InitializeParticle(
    Color=&LerpLinearColors(StartColor=<rgba 0.01 0.01 0.01 2>, EndColor=<rgba 0.15 0 0.3 1>, LerpFactor=$User.VariantSwitch)
    Lifetime Max=1.5
    Lifetime Mode=::NewEnumerator1
    Sprite Rotation Mode=::NewEnumerator2
    Sprite Size Mode=::NewEnumerator1
    Uniform Sprite Size Max=300
    Uniform Sprite Size Min=200
    UsePositionOffset=false
    Write Material Random=true
   )
   &AddVelocity(Velocity=<0 -500 0>)
   &InheritVelocity(Inherited Velocity Amount Scale=<1 1 0>, Inherited Velocity Speed Limit=1000, Speed Limit=true, UseSpeedThreshold=false)
  ]
  particleUpdateScript [
   &CameraOffset(Camera Offset Amount=-50)
   &SkeletalMeshLocation(
    Bone / Socket Index=4
    Bone Sampling Mode=::NewEnumerator4
    MaskSpawnGroup=false
    Orientation Sampling=::NewEnumerator1
    Position Sampling=::NewEnumerator1
    Sampled Position Offset=<0 0 -50>
    Sampled Position Offset Coordinate Space:ENiagaraSkelMeshTransforms=::NewEnumerator2
    Skeletal Mesh=$System.Skeletal Mesh
    Source Movement Scale:Vector3f=<0.2 0.2 1>
   )
   &ParticleState()
   &ScaleSpriteSize(Uniform Curve Sprite Scale=DataInterface() {propertyValues {Curve {keys [
        {arriveTangent=1.6237, interpMode=RCIM_Cubic, leaveTangent=1.6237, tangentMode=RCTM_User, time=0, value=1}
        {arriveTangent=0.637, tangentMode=RCTM_User, time=1, value=1.7}
       ]}}})
   &SpriteRotationRate(Rotation Rate=&RandomRangeFloat(Minimum=-150, Maximum=150))
   &SolveForcesAndVelocity(Clamp Velocity=false, Limit Acceleration=false)
   &ScaleColor(
    Scale Alpha=&FloatFromCurve(FloatCurve=DataInterface() {propertyValues {Curve {keys [
         {interpMode=RCIM_Cubic, tangentMode=RCTM_User, time=0, value=0}
         {arriveTangent=-0.5, interpMode=RCIM_Cubic, leaveTangent=-3.148, tangentMode=RCTM_Break, time=0.2, value=0.85}
         <at 1 0 RCIM_Cubic>
        ]}}})
    ScaleA=true
    ScaleRGB=true
   )
  ]
  rendererClasses [&NiagaraSpriteRendererProperties]
  renderers [{rendererClass=&NiagaraSpriteRendererProperties}]
 }
 FE_Chariot_Green_Smoke2 {
  emitterUpdateScript [
   &EmitterState(
    Life Cycle Mode=::NewEnumerator1
    Loop Behavior=::NewEnumerator1
    Loop Delay=$System.Delay
    Loop Duration=$System.Duration
    UseLoopDelay=true
   )
   &SpawnRate(
    Spawn Probability=&MakeFloatFromInt(Int=&MakeCustomIntFromBool(Bool=$User.bIsFrontEndPreview, True INT=1, False INT=0))
    SpawnRate=7
    Use Spawn Probability=true
   )
  ]
  particleSpawnScript [
   &PartitionParticles(Number Of Partitions=1)
   &InitializeParticle(
    Color=&LerpLinearColors(StartColor=<rgba 0.0053 0.04 0.0169 0.7>, EndColor=<rgba 0.055 0.01 0.1 0.7>, LerpFactor=$User.VariantSwitch)
    Lifetime Min=2
    Lifetime Mode=::NewEnumerator1
    Sprite Rotation Mode=::NewEnumerator2
    Sprite Size Mode=::NewEnumerator1
    Sprite UV Mode=::NewEnumerator3
    Uniform Sprite Size Max=150
    Uniform Sprite Size Min=140
    UsePositionOffset=false
    Write Material Random=true
   )
   &AddVelocity(Velocity=<-150 -500 0>, Velocity Speed Scale=1.5)
   &CameraOffset(Camera Offset Amount=-50)
  ]
  particleUpdateScript [
   &SkeletalMeshLocation(
    Bone / Socket Index=4
    Bone Sampling Mode=::NewEnumerator4
    MaskSpawnGroup=false
    Orientation Sampling=::NewEnumerator1
    Position Sampling=::NewEnumerator1
    Sampled Position Offset=<0 0 -50>
    Sampled Position Offset Coordinate Space:ENiagaraSkelMeshTransforms=::NewEnumerator2
    Skeletal Mesh=$System.Skeletal Mesh
    Source Movement Scale:Vector3f=<0 0 0>
   )
   &ParticleState()
   &ScaleSpriteSize(
    Uniform Curve Scale=&MakeFloatFromVector(Vector=$Engine.Owner.Scale, Channel=::NewEnumerator0)
    Uniform Curve Sprite Scale=DataInterface() {propertyValues {Curve {keys [
        {arriveTangent=0.2792, interpMode=RCIM_Cubic, leaveTangent=0.2792, tangentMode=RCTM_User, time=0, value=1.6}
        {arriveTangent=0.637, tangentMode=RCTM_User, time=1, value=1.7}
       ]}}}
   )
   &SpriteRotationRate(Rotation Rate=&RandomRangeFloat(Minimum=-150, Maximum=150))
   &SolveForcesAndVelocity(Clamp Velocity=false, Limit Acceleration=false)
   &ScaleColor(
    Scale Alpha=&FloatFromCurve(FloatCurve=DataInterface() {propertyValues {Curve {keys [
         {interpMode=RCIM_Cubic, tangentMode=RCTM_User, time=0, value=0}
         {arriveTangent=-0.5, interpMode=RCIM_Cubic, leaveTangent=-3.148, tangentMode=RCTM_Break, time=0.1, value=0.85}
         <at 0.4 0 RCIM_Cubic>
        ]}}})
    ScaleA=true
    ScaleRGB=true
   )
   &SubUVAnimation(
    End Frame Range Override=64
    "Sprite Renderer "=DataInterface() {propertyValues {SpriteRenderer=&self:FE_Chariot_Green_Smoke2_2.NiagaraSpriteRendererProperties_1}}
    SubUV Lookup Index Scale=1.5
    UseEndFrame=true
    UseStartFrame=true
   )
  ]
  rendererClasses [&NiagaraSpriteRendererProperties]
  renderers [{rendererClass=&NiagaraSpriteRendererProperties}]
 }
 FE_Chariot_Smoke {
  emitterUpdateScript [
   &EmitterState(
    Life Cycle Mode=::NewEnumerator1
    Loop Behavior=::NewEnumerator1
    Loop Delay=$System.Delay
    Loop Duration=$System.Duration
    UseLoopDelay=true
   )
   &SpawnRate(
    Spawn Probability=&MakeFloatFromInt(Int=&MakeCustomIntFromBool(Bool=$User.bIsFrontEndPreview, True INT=1, False INT=0))
    SpawnRate=10
    Use Spawn Probability=true
   )
  ]
  particleSpawnScript [
   &PartitionParticles(Number Of Partitions=1)
   &InitializeParticle(
    Color=&LerpLinearColors(StartColor=<rgba 0 0.4844 0.1615 1>, EndColor=<rgba 0.36 0.12 0.6 1>, LerpFactor=$User.VariantSwitch)
    Lifetime Max=0.7
    Lifetime Min=0.7
    Lifetime Mode=::NewEnumerator1
    Sprite Rotation Mode=::NewEnumerator2
    Sprite Size Mode=::NewEnumerator1
    Uniform Sprite Size Max=300
    Uniform Sprite Size Min=200
    UsePositionOffset=false
    Write Material Random=true
   )
   &AddVelocity(Velocity=<-150 -500 0>, Velocity Speed Scale=1.5)
   &CameraOffset(Camera Offset Amount=-50)
  ]
  particleUpdateScript [
   &SkeletalMeshLocation(
    Bone / Socket Index=4
    Bone Sampling Mode=::NewEnumerator4
    MaskSpawnGroup=false
    Orientation Sampling=::NewEnumerator1
    Position Sampling=::NewEnumerator1
    Sampled Position Offset=<0 0 -50>
    Sampled Position Offset Coordinate Space:ENiagaraSkelMeshTransforms=::NewEnumerator2
    Skeletal Mesh=$System.Skeletal Mesh
    Source Movement Scale:Vector3f=<0 0 0>
   )
   &ParticleState()
   &ScaleSpriteSize(
    Uniform Curve Scale=&MakeFloatFromVector(Vector=$Engine.Owner.Scale, Channel=::NewEnumerator0)
    Uniform Curve Sprite Scale=DataInterface() {propertyValues {Curve {keys [
        {arriveTangent=1.6237, interpMode=RCIM_Cubic, leaveTangent=1.6237, tangentMode=RCTM_User, time=0, value=1}
        {arriveTangent=0.637, tangentMode=RCTM_User, time=1, value=1.7}
       ]}}}
   )
   &SpriteRotationRate(Rotation Rate=&RandomRangeFloat(Minimum=-150, Maximum=150))
   &SolveForcesAndVelocity(Clamp Velocity=false, Limit Acceleration=false)
   &ScaleColor(
    Scale Alpha=&FloatFromCurve(FloatCurve=DataInterface() {propertyValues {Curve {keys [
         {interpMode=RCIM_Cubic, tangentMode=RCTM_User, time=0, value=0}
         {arriveTangent=-0.5, interpMode=RCIM_Cubic, leaveTangent=-3.148, tangentMode=RCTM_Break, time=0.2, value=0.85}
         <at 1 0 RCIM_Cubic>
        ]}}})
    ScaleA=true
    ScaleRGB=true
   )
  ]
  rendererClasses [&NiagaraSpriteRendererProperties]
  renderers [{rendererClass=&NiagaraSpriteRendererProperties}]
 }
 FE_Chariot_Smoke001 {
  emitterUpdateScript [
   &EmitterState(
    Life Cycle Mode=::NewEnumerator1
    Loop Behavior=::NewEnumerator1
    Loop Delay=$System.Delay
    Loop Duration=$System.Duration
    UseLoopDelay=true
   )
   &SpawnRate(
    Spawn Probability=&MakeFloatFromInt(Int=&MakeCustomIntFromBool(Bool=$User.bIsFrontEndPreview, True INT=1, False INT=0))
    SpawnRate=15
    Use Spawn Probability=true
   )
  ]
  particleSpawnScript [
   &PartitionParticles(Number Of Partitions=1)
   &InitializeParticle(
    Color=&LerpLinearColors(StartColor=<rgba 0.01 0.01 0.01 2>, EndColor=<rgba 0.1 0 0.2 1>, LerpFactor=$User.VariantSwitch)
    Lifetime Max=1.5
    Lifetime Mode=::NewEnumerator1
    Sprite Rotation Mode=::NewEnumerator2
    Sprite Size Mode=::NewEnumerator1
    Uniform Sprite Size Max=400
    Uniform Sprite Size Min=200
    UsePositionOffset=false
    Write Material Random=true
   )
   &AddVelocity(Velocity=<-150 -500 0>, Velocity Speed Scale=1.5)
   &CameraOffset(Camera Offset Amount=-50)
  ]
  particleUpdateScript [
   &SkeletalMeshLocation(
    Bone / Socket Index=4
    Bone Sampling Mode=::NewEnumerator4
    MaskSpawnGroup=false
    Orientation Sampling=::NewEnumerator1
    Position Sampling=::NewEnumerator1
    Sampled Position Offset=<0 0 -50>
    Sampled Position Offset Coordinate Space:ENiagaraSkelMeshTransforms=::NewEnumerator2
    Skeletal Mesh=$System.Skeletal Mesh
    Source Movement Scale:Vector3f=<0 0 0>
   )
   &ParticleState()
   &ScaleSpriteSize(
    Uniform Curve Scale=&MakeFloatFromVector(Vector=$Engine.Owner.Scale, Channel=::NewEnumerator0)
    Uniform Curve Sprite Scale=DataInterface() {propertyValues {Curve {keys [
        {arriveTangent=1.6237, interpMode=RCIM_Cubic, leaveTangent=1.6237, tangentMode=RCTM_User, time=0, value=1}
        {arriveTangent=0.637, tangentMode=RCTM_User, time=1, value=1.7}
       ]}}}
   )
   &SpriteRotationRate(Rotation Rate=&RandomRangeFloat(Minimum=-150, Maximum=150))
   &SolveForcesAndVelocity(Clamp Velocity=false, Limit Acceleration=false)
   &ScaleColor(
    Scale Alpha=&FloatFromCurve(FloatCurve=DataInterface() {propertyValues {Curve {keys [
         {interpMode=RCIM_Cubic, tangentMode=RCTM_User, time=0, value=0}
         {arriveTangent=-0.5, interpMode=RCIM_Cubic, leaveTangent=-3.148, tangentMode=RCTM_Break, time=0.2, value=0.85}
         <at 1 0 RCIM_Cubic>
        ]}}})
    ScaleA=true
    ScaleRGB=true
   )
  ]
  rendererClasses [&NiagaraSpriteRendererProperties]
  renderers [{rendererClass=&NiagaraSpriteRendererProperties}]
 }
 FE_Horse_Smoke {
  emitterUpdateScript [
   &EmitterState(
    Delay First Loop Only=true
    Inactive Response=::NewEnumerator1
    Life Cycle Mode=::NewEnumerator1
    Loop Delay=$System.Delay
    Loop Duration=1
    UseLoopDelay=true
   )
   &SpawnRate(
    Spawn Probability=&MakeFloatFromInt(Int=&MakeCustomIntFromBool(Bool=$User.bIsFrontEndPreview, True INT=1, False INT=0))
    SpawnRate=20
    Use Spawn Probability=true
   )
  ]
  particleSpawnScript [
   &PartitionParticles(Number Of Partitions=4)
   &InitializeParticle(
    Color=&LerpLinearColors(StartColor=<rgba 0.01 0.01 0.01 0.7>, EndColor=<rgba 0.165 0.03 0.3 0.7>, LerpFactor=$User.VariantSwitch)
    Lifetime Max=1
    Lifetime Min=0.5
    Lifetime Mode=::NewEnumerator1
    Sprite Rotation Mode=::NewEnumerator2
    Sprite Size Mode=::NewEnumerator1
    Uniform Sprite Size Max=200
    Uniform Sprite Size Min=100
    UsePositionOffset=false
    Write Material Random=true
   )
   &AddVelocity(Velocity=<0 -500 -200>)
   &CameraOffset(Camera Offset Amount=-50)
  ]
  particleUpdateScript [
   &SkeletalMeshLocation(
    Bone / Socket Index=$Particles.PartitionParticles.Partition
    Bone Sampling Mode=::NewEnumerator4
    MaskSpawnGroup=false
    Orientation Sampling=::NewEnumerator1
    Position Sampling=::NewEnumerator1
    Sampled Position Offset=<0 0 -20>
    Sampled Position Offset Coordinate Space:ENiagaraSkelMeshTransforms=::NewEnumerator2
    Skeletal Mesh=$System.Skeletal Mesh
    Source Movement Scale:Vector3f=<0 0 0>
   )
   &ParticleState()
   &ScaleSpriteSize(
    Uniform Curve Scale=&MakeFloatFromVector(Vector=$Engine.Owner.Scale, Channel=::NewEnumerator0)
    Uniform Curve Sprite Scale=DataInterface() {propertyValues {Curve {keys [
        {arriveTangent=1.6237, interpMode=RCIM_Cubic, leaveTangent=1.6237, tangentMode=RCTM_User, time=0, value=1}
        {arriveTangent=0.637, tangentMode=RCTM_User, time=1, value=1.7}
       ]}}}
   )
   &SpriteRotationRate(Rotation Rate=&RandomRangeFloat(Minimum=-150, Maximum=150))
   &ScaleVelocity(Velocity Scale=&MakeVector(
     X=1
     Y=1
     Z=&FloatFromCurve(
      FloatCurve=DataInterface() {propertyValues {Curve {keys [{tangentMode=RCTM_User, time=2.5, value=0}, {tangentMode=RCTM_User, time=3, value=1}]}}}
      CurveIndex=$Emitter.Age
      Scale Curve=&Lerp_Float(A=1, B=1.5, Alpha=$System.InterpolateOverTime.InterpolateOverTime_Float.Moving Average)
     )
    ))
   &Drag(Drag=1.5, Use Linear Drag=true, Use Rotational Drag=true)
   &SolveForcesAndVelocity(Clamp Velocity=false, Limit Acceleration=false)
   &ScaleColor(
    Scale Alpha=&FloatFromCurve(FloatCurve=DataInterface() {propertyValues {Curve {keys [
         {arriveTangent=3.2632, interpMode=RCIM_Cubic, leaveTangent=3.2632, tangentMode=RCTM_User, time=0, value=0}
         {arriveTangent=-0.5, interpMode=RCIM_Cubic, leaveTangent=-3.148, tangentMode=RCTM_Break, time=0.3, value=0.85}
         <at 1 0 RCIM_Cubic>
        ]}}})
    ScaleA=true
    ScaleRGB=true
   )
  ]
  rendererClasses [&NiagaraSpriteRendererProperties]
  renderers [{rendererClass=&NiagaraSpriteRendererProperties}]
 }
 Horse_Smoke {
  emitterUpdateScript [
   &EmitterState(Inactive Response=::NewEnumerator1, Life Cycle Mode=::NewEnumerator1, Loop Duration=1, UseLoopDelay=false)
   &SpawnRate(
    Spawn Probability=&MakeFloatFromInt(Int=&MakeCustomIntFromBool(Bool=$User.bIsFrontEndPreview, True INT=0, False INT=1))
    SpawnRate=&Multiply_Float(A=20, B=$System.InterpolateOverTime.InterpolateOverTime_Float.Moving Average)
    Use Spawn Probability=true
   )
  ]
  particleSpawnScript [
   &PartitionParticles(Number Of Partitions=4)
   &InitializeParticle(
    Color=&LerpLinearColors(StartColor=<rgba 0.01 0.01 0.01 1>, EndColor=<rgba 0.1 0 0.2 1>, LerpFactor=$User.VariantSwitch)
    Lifetime Max=1
    Lifetime Min=0.5
    Lifetime Mode=::NewEnumerator1
    Sprite Rotation Mode=::NewEnumerator2
    Sprite Size Mode=::NewEnumerator1
    Uniform Sprite Size Max=200
    Uniform Sprite Size Min=150
    UsePositionOffset=false
    Write Material Random=true
   )
   &AddVelocity(Velocity=<0 -500 -200>)
   &CameraOffset(Camera Offset Amount=-50)
  ]
  particleUpdateScript [
   &SkeletalMeshLocation(
    Bone / Socket Index=$Particles.PartitionParticles.Partition
    Bone Sampling Mode=::NewEnumerator4
    MaskSpawnGroup=false
    Orientation Sampling=::NewEnumerator1
    Position Sampling=::NewEnumerator1
    Sampled Position Offset=<0 0 -35>
    Sampled Position Offset Coordinate Space:ENiagaraSkelMeshTransforms=::NewEnumerator2
    Skeletal Mesh=$System.Skeletal Mesh
    Source Movement Scale:Vector3f=<0 0 0>
   )
   &ParticleState()
   &ScaleSpriteSize(Uniform Curve Sprite Scale=DataInterface() {propertyValues {Curve {keys [
        {arriveTangent=1.6237, interpMode=RCIM_Cubic, leaveTangent=1.6237, tangentMode=RCTM_User, time=0, value=1}
        {arriveTangent=0.637, tangentMode=RCTM_User, time=1, value=1.7}
       ]}}})
   &SpriteRotationRate(Rotation Rate=&RandomRangeFloat(Minimum=-150, Maximum=150))
   &ScaleVelocity(Velocity Scale=&MakeVector(
     X=1
     Y=1
     Z=&FloatFromCurve(
      FloatCurve=DataInterface() {propertyValues {Curve {keys [{tangentMode=RCTM_User, time=2.5, value=0}, {tangentMode=RCTM_User, time=3, value=1}]}}}
      CurveIndex=$Emitter.Age
      Scale Curve=&Lerp_Float(A=1, B=1.5, Alpha=$System.InterpolateOverTime.InterpolateOverTime_Float.Moving Average)
     )
    ))
   &Drag(Drag=1.5, Use Linear Drag=true, Use Rotational Drag=true)
   &SolveForcesAndVelocity(Clamp Velocity=false, Limit Acceleration=false)
   &ScaleColor(
    Scale Alpha=&FloatFromCurve(FloatCurve=DataInterface() {propertyValues {Curve {keys [
         {arriveTangent=3.2632, interpMode=RCIM_Cubic, leaveTangent=3.2632, tangentMode=RCTM_User, time=0, value=0}
         {arriveTangent=-0.5, interpMode=RCIM_Cubic, leaveTangent=-3.148, tangentMode=RCTM_Break, time=0.3, value=0.85}
         <at 1 0 RCIM_Cubic>
        ]}}})
    ScaleA=true
    ScaleRGB=true
   )
  ]
  rendererClasses [&NiagaraSpriteRendererProperties]
  renderers [{rendererClass=&NiagaraSpriteRendererProperties}]
 }
 Sparkles {
  bEnabled=false
  simTarget=GPUComputeSim
  emitterUpdateScript [
   &EmitterState(
    Inactive Response=::NewEnumerator1
    Life Cycle Mode=::NewEnumerator1
    Loop Behavior=::NewEnumerator1
    Loop Duration=$System.Duration
    UseLoopDelay=false
   )
   &SpawnRate(
    Spawn Probability=&MakeFloatFromInt(Int=&MakeCustomIntFromBool(Bool=$User.bIsFrontEndPreview, True INT=0, False INT=1))
    SpawnRate=10
    Use Spawn Probability=true
   )
  ]
  particleSpawnScript [
   &PartitionParticles(Number Of Partitions=1)
   &InitializeParticle(
    Color=&LerpLinearColors(StartColor=<rgba 0 5 1.1667 1>, EndColor=<rgba 5 0 10 1>, LerpFactor=$User.VariantSwitch)
    Lifetime Mode=::NewEnumerator1
    Sprite Rotation Mode=::NewEnumerator2
    Sprite Size Max=<1 4>
    Sprite Size Min=<1 4>
    Sprite Size Mode=::NewEnumerator2
    UsePositionOffset=false
    Write Material Random=true
   )
   &ShapeLocation(
    MaskSpawnGroup=false
    Override Seed=false
    Shape Origin=&VectorFromFloat(Value=0)
    Sphere Radius=50
    UseOwnerScale=true
   )
   &InheritVelocity(Inherited Velocity Amount Scale=<1 1 0>, Inherited Velocity Speed Limit=600, Speed Limit=true, UseSpeedThreshold=false)
   &CameraOffset(Camera Offset Amount=-50)
  ]
  particleUpdateScript [
   &SkeletalMeshLocation(
    Bone / Socket Index=4
    Bone Sampling Mode=::NewEnumerator4
    MaskSpawnGroup=false
    Orientation Sampling=::NewEnumerator1
    Position Sampling=::NewEnumerator1
    Sampled Position Offset=$Particles.ShapeLocation.ShapePosition
    Sampled Position Offset Coordinate Space:ENiagaraSkelMeshTransforms=::NewEnumerator2
    Skeletal Mesh=$System.Skeletal Mesh
    Source Movement Scale:Vector3f=<0.2 0.2 1>
   )
   &ParticleState()
   &ScaleSpriteSize(
    Scale Sprite Size Mode=::NewEnumerator0
    Uniform Scale Factor=&MakeFloatFromVector2d(Vector 2D=&Vector2DFromFloat(Value=&FloatFromCurve(
       FloatCurve=DataInterface() {propertyValues {Curve {keys [
           <at 0 0 RCIM_Cubic>
           {interpMode=RCIM_Cubic, leaveTangent=-3.0296, tangentMode=RCTM_Break, time=0.2, value=1.2}
           {arriveTangent=-1.8107, interpMode=RCIM_Cubic, leaveTangent=-1.8107, tangentMode=RCTM_User, time=1, value=0}
          ]}}}
       Scale Curve=2
      )))
   )
   &SpriteRotationRate(Rotation Rate=&RandomRangeFloat(Minimum=-150, Maximum=150))
   &CurlNoiseForce(Noise Frequency=500, Noise Strength=1000, Pan Noise=false)
   &SolveForcesAndVelocity(Clamp Velocity=false, Limit Acceleration=false)
   &ScaleColor(
    Scale Alpha=&FloatFromCurve(FloatCurve=DataInterface() {propertyValues {Curve {keys [
         {interpMode=RCIM_Cubic, tangentMode=RCTM_User, time=0, value=0}
         {arriveTangent=-0.5, interpMode=RCIM_Cubic, leaveTangent=-3.148, tangentMode=RCTM_Break, time=0.2, value=0.85}
         <at 1 0 RCIM_Cubic>
        ]}}})
    ScaleA=true
    ScaleRGB=true
   )
   &FN_Sprite_Attributes(
    Color=&MakeLinearColorFromVectorAndFloat(
     "Vector (RGB)"=&MakeVectorFromLinearColorRGB(Color=&MultiplyLinearColor(
       Linear Color=&LerpLinearColors(StartColor=<rgba 0.0335 1 0.2591 1>, EndColor=<rgba 1.0335 0.0671 2 1>, LerpFactor=$User.VariantSwitch)
       Scale Factor=&MakeVector4_FromFloat(Float=1)
      ))
     "Float (Alpha)"=&Multiply_Float(A=0.2, B=1)
    )
    Create Color Attribute=true
    Create SpriteSize Attribute=true
    SpriteSize=&Multiply_Vector2D_ByFloat(
     Float=&FloatFromCurve(
      FloatCurve=DataInterface() {propertyValues {Curve {keys [
          <at 0 0 RCIM_Cubic>
          <at 0.2 1.3 RCIM_Cubic>
          {arriveTangent=-0.1423, interpMode=RCIM_Cubic, leaveTangent=-0.1423, tangentMode=RCTM_User, time=1, value=0}
         ]}}}
      Scale Curve=4
     )
     Vector2D=<3 5>
    )
    Write to Module Namespace=true
   )
  ]
  rendererClasses [&NiagaraSpriteRendererProperties]
  renderers [
   {rendererClass=&NiagaraSpriteRendererProperties}
   {rendererClass=&NiagaraSpriteRendererProperties, rendererIndex=1}
  ]
 }
}
summary {
 systemName=NS_Glider_DarkStance_Deploy
 emitters [
  {emitterName=Chariot_Green_Smoke2, rendererClasses [&NiagaraSpriteRendererProperties]}
  {emitterName=Chariot_Green_Smoke, simTarget=GPUComputeSim, rendererClasses [&NiagaraSpriteRendererProperties]}
  {bEnabled=false, emitterName=Sparkles, simTarget=GPUComputeSim, rendererClasses [&NiagaraSpriteRendererProperties]}
  {emitterName=Chariot_Smoke001, rendererClasses [&NiagaraSpriteRendererProperties]}
  {emitterName=Horse_Smoke, rendererClasses [&NiagaraSpriteRendererProperties]}
  {emitterName=FE_Chariot_Green_Smoke2, rendererClasses [&NiagaraSpriteRendererProperties]}
  {emitterName=FE_Chariot_Smoke001, rendererClasses [&NiagaraSpriteRendererProperties]}
  {emitterName=FE_Horse_Smoke, rendererClasses [&NiagaraSpriteRendererProperties]}
  {emitterName=FE_Chariot_Smoke, rendererClasses [&NiagaraSpriteRendererProperties]}
 ]
 userVariables [
  User.bIsFrontEndPreview {type=NiagaraBool}
  User.ForwardDot {type=NiagaraFloat, defaultValue=0, description=""}
  User.VariantSwitch {type=NiagaraFloat, defaultValue=0, description=""}
 ]
}
